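; BPI_GamePhaseWidget is a Blueprint interface for Athena HUD widgets that present the current game phase.
; It declares phase-step functions for step changes with an active icon index, intro animations, and messaging with time remaining and an optional siphon text flag.
; Other functions cover the jump animation, the playlist name, and ranked progress by key, tier, position, percentage and MMR.
; It holds only signatures, so implementing widgets supply all behavior.
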(fn HandleGamePhaseStepChanged
  (params (Step :type EAthenaGamePhaseStep%20Enum) (ActiveIconIndex :type Integer)))

(fn HandleJumpAnimation)

(fn PlayIntroAnimations
  (params (Step :type EAthenaGamePhaseStep%20Enum)))

(fn SetPlaylistName
  (params (PlaylistName :type Text)))

(fn SetRankedProgress
  (params (ProgressKey :type String) (RankedTier :type Integer) (RankedPosition :type Integer) (RankedPercentage :type Float%20%28double-precision%29) (RankedMMR :type Integer)))

(fn UpdateMessaging
  (params (Step :type EAthenaGamePhaseStep%20Enum) (Message :type Text) (TimeRemaining :type Text) (ShowSiphonText :type Boolean)))
