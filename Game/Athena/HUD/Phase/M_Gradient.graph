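# Translucent UI material in the Athena HUD phase folder that draws a gradient, either reflected or radial.
# The IsRadial scalar, defaulting to radial, blends MF_UI_ReflectedGradient with RadialGradientExponential.
# The radial branch centers on the top-middle edge and divides V by four, stretching the falloff vertically into an ellipse.
# A Color vector parameter, white by default, is also declared.
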Material /Game/Athena/HUD/Phase/M_Gradient
graph {
 properties {bAutomaticallySetUsageInEditor=false, blendMode=BLEND_Translucent, materialDomain=MD_UI, usedShadingModels=Default Lit}
 nodes [
  n1 = TextureCoordinate()
  n2 = ComponentMask(None=@n1) {r=true}
  n3 = Divide(A=ComponentMask(None=@n1) {g=true}) {constB=4}
  LinearInterpolate(
   A=MaterialFunctionCall!(UVs, Offset, RotationCenter, RotationAngle) {materialFunction=/FortUILibrary/MaterialFunctions/Gradients/MF_UI_ReflectedGradient}.Result
   B=MaterialFunctionCall(UVs=AppendVector(A=@n2, B=@n3), CenterPosition=Constant2Vector() {r=0.5}, Radius=Constant() {r=0.5}, Density=Constant() {r=2}) {materialFunction=&RadialGradientExponential}.RadialGradientExponential
   Alpha=ScalarParameter() {defaultValue=1, parameterName=IsRadial}
  )
  VectorParameter() {parameterName=Color, defaultValue=<rgba 1 1 1 1>}
 ]
}
