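# Translucent UI material for the Athena HUD, named for a phase change time that is soon, producing a pulsing white brightness.
# A pause-independent Time node with a one-second period is remapped to a signed range and fed through the engine BellCurve function.
# The curve drives a lerp between dim and overbright levels, which multiplies white to give a smooth once-per-second pulse without textures or parameters.
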Material /Game/Athena/HUD/Phase/M_PhaseChangeTimeSoon
graph {
 properties {blendMode=BLEND_Translucent, materialDomain=MD_UI, pixelDepthOffsetMode=PDOM_Legacy, usedShadingModels=Default Lit}
 nodes [
  n1 = Time() {bIgnorePause=true, bOverride_Period=true, period=1}
  n2 = MaterialFunctionCall!(x=Subtract(A=Multiply(A=@n1) {constB=2})) {materialFunction=/Engine/Functions/Engine_MaterialFunctions01/Gradient/BellCurve}
  Multiply(A=Multiply(A=LinearInterpolate(Alpha=@n2.0-1) {constA=0.5, constB=2}, B=Constant3Vector() {constant=<rgba 1 1 1 1>}))
 ]
}
