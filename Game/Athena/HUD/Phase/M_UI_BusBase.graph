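# Translucent UI material for the bus base in the Athena HUD phase display, drawing a skewed panel with animated smoky clouds and a gradient outline.
# Two panned noise textures, faded by a powered vertical gradient, blend a cloud color over the background color.
# Outer and inner SkewBox shapes are smoothstepped and subtracted to mask the two-color outline.
# The inner box scale uses the aspect ratio from texture coordinate 3 to compensate border thickness.
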Material /Game/Athena/HUD/Phase/M_UI_BusBase
graph {
 properties {blendMode=BLEND_Translucent, materialDomain=MD_UI, pixelDepthOffsetMode=PDOM_Legacy, usedShadingModels=Default Lit}
 nodes [
  n1 = VectorParameter() {parameterName=Color Clouds, defaultValue=<rgba 0.6643 0.1292 1 1>}
  n2 = MaterialFunctionCall() {materialFunction=&MF_UI_UVScaling}
  n3 = Multiply(A=Time(), B=ScalarParameter() {defaultValue=1, parameterName=Smoke Speed})
  n4 = Constant2Vector() {g=0.02, r=-0.1}
  n5 = Panner(Coordinate=@n2.ScaleX, Time=@n3, Speed=Constant2Vector() {g=0.2, r=-0.2}) {speedX=-0.3, speedY=0.2}
  n6 = Add(
   A=TextureSample(UVs=Panner(Coordinate=@n2.ScaleX, Time=@n3, Speed=@n4) {speedX=-0.3, speedY=0.2}) {samplerSource=SSM_Wrap_WorldGroupSettings, texture=&/T_LowRes_Noise_Phase}.RGB
   B=Multiply(A=TextureSample(UVs=@n5) {samplerSource=SSM_Wrap_WorldGroupSettings, samplerType=SAMPLERTYPE_Grayscale, texture=/FortWater/Textures/LowResBlurredNoise}.RGB) {constB=0.3}
  )
  n7 = Power(Base=MaterialFunctionCall() {materialFunction=&LinearGradient}.VGradient) {constExponent=5}
  n8 = Multiply(A=Saturate(None=LinearInterpolate(B=@n6, Alpha=@n7)))
  n9 = LinearInterpolate(
   A=VectorParameter() {parameterName=Color BG, defaultValue=<rgba 0.0239 0.0156 0.375 1>}.RGB
   B=MaterialFunctionCall(Base=@n1.RGB, Blend=@n1.RGB) {materialFunction=&Blend_Overlay}.Result
   Alpha=@n8
  )
  n10 = Multiply(A=@n1.RGB, B=@n8)
  n11 = Constant()
  n12 = Multiply(A=ScalarParameter() {defaultValue=0.01, parameterName=Smoothness}, B=NamedRerouteUsage())
  n13 = ScalarParameter() {parameterName=MinWidth}
  n14 = SmoothStep(
   Min=@n11
   Max=@n12
   Value=MaterialFunctionCall!(
    MinWidth=@n13
    Angle Top=ScalarParameter() {defaultValue=0.05, group=Outer Skew Box, parameterName=O_Angle_Top}
    Angle Right=ScalarParameter() {defaultValue=0.1, group=Outer Skew Box, parameterName=O_Angle_Right}
    Angle Bottom=ScalarParameter() {defaultValue=0.05, group=Outer Skew Box, parameterName=O_Angle_Bottom}
    Angle Left=ScalarParameter() {defaultValue=0.2, group=Outer Skew Box, parameterName=O_Angle_Left}
    Scale=ScalarParameter() {defaultValue=1, group=Outer Skew Box, parameterName=O_SkewScale}
    Flip=StaticBool() {value=true}
   ) {materialFunction=&/MF_UI_Ranked_SkewBox}.Result
  )
  n15 = VectorParameter() {parameterName=Outline Color 1, defaultValue=<rgba 0.0578 0.007 0.1095 1>}
  n16 = VectorParameter() {parameterName=Outline Color 2, defaultValue=<rgba 0.7835 0.4179 1 1>}
  n17 = ScalarParameter() {defaultValue=0.07, group=Inner Skew Box, parameterName=I_SkewScale}
  n18 = TextureCoordinate() {coordinateIndex=3}
  n19 = ComponentMask(None=@n18) {g=true}
  n20 = ComponentMask(None=@n18) {r=true}
  n21 = Divide(A=@n19, B=@n20)
  n22 = Divide(A=@n20, B=@n19)
  n23 = Saturate(None=AppendVector(A=@n21, B=@n22))
  n24 = MaterialFunctionCall(
   MinWidth=@n13
   Angle Top=ScalarParameter() {defaultValue=0.05, group=Inner Skew Box, parameterName=I_Angle_Top}
   Angle Right=ScalarParameter() {defaultValue=0.1, group=Inner Skew Box, parameterName=I_Angle_Right}
   Angle Bottom=ScalarParameter() {defaultValue=0.05, group=Inner Skew Box, parameterName=I_Angle_Bottom}
   Angle Left=ScalarParameter() {defaultValue=0.2, group=Inner Skew Box, parameterName=I_Angle_Left}
   Scale=OneMinus(None=Multiply(A=@n17, B=@n23))
   Flip=StaticBool() {value=true}
  ) {materialFunction=&MF_UI_Ranked_SkewBox}
  LinearInterpolate(
   A=Multiply(A=Add(A=@n9, B=@n10), B=@n14)
   B=LinearInterpolate(
    A=ComponentMask(None=@n15.RGB) {b=true, g=true, r=true}
    B=ComponentMask(None=@n16.RGB) {b=true, g=true, r=true}
    Alpha=MaterialFunctionCall() {materialFunction=&LinearGradient}.VGradient
   )
   Alpha=Subtract(A=@n14, B=SmoothStep(Min=@n11, Max=@n12, Value=@n24.Result))
  )
  NamedRerouteDeclaration(None=@n21) {name=ScaleX, nodeColor=<rgba 0 0 0 1>}
  NamedRerouteDeclaration(None=@n22) {name=ScaleY, nodeColor=<rgba 0 0 0 1>}
  Reroute(None=@n14)
  ScalarParameter() {group=Inner Skew Box, parameterName=I_Offset_Right}
 ]
}
