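# Translucent UI material that draws the background of the HUD phase display, with parameters for placement, skew, color and a bottom variant.
# The shape is the max of four skewed rounded masks, mirrored horizontally, with IsBottom blending in an extra mask.
# Color adds a glow from two squashed radial gradients, faded out by IsBottom.
# Named reroutes carry a delayed, eased highlight sweep, a radial-angle ray shape and a stepped, angle-thresholded center alpha.
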Material /Game/Athena/HUD/Phase/M_UI_HUD_PhaseBackground
graph {
 properties {bAutomaticallySetUsageInEditor=false, blendMode=BLEND_Translucent, materialDomain=MD_UI, usedShadingModels=Default Lit}
 nodes [
  n1 = VectorParameter() {parameterName=Place, defaultValue=<rgba 0.2 0 -0.2 0>}
  n2 = TextureCoordinate()
  n3 = ComponentMask(None=@n2) {r=true}
  n4 = ComponentMask(None=@n2) {g=true}
  n5 = Add(A=@n4, B=@n1.G) {constB=0.2}
  n6 = Saturate(None=AppendVector(A=Add(A=@n3, B=@n1.R) {constB=0.2}, B=@n5))
  n7 = Constant()
  n8 = ScalarParameter() {defaultValue=45, parameterName=Skew}
  n9 = VectorParameter() {parameterName=Size, defaultValue=<rgba 0.2 0.2 0.2 0.2>}
  n10 = AppendVector(A=@n9.R, B=@n9.R)
  n11 = Add(A=@n4, B=@n1.A) {constB=0.2}
  n12 = Saturate(None=AppendVector(A=Add(A=@n3, B=@n1.B) {constB=0.2}, B=@n11))
  n13 = AppendVector(A=@n9.B, B=@n9.B)
  n14 = OneMinus(None=@n3)
  n15 = Saturate(None=AppendVector(A=Add(A=@n14, B=@n1.R) {constB=0.2}, B=@n5))
  n16 = Saturate(None=AppendVector(A=Add(A=@n14, B=@n1.B) {constB=0.2}, B=@n11))
  n17 = Max(
   A=Max(
    A=MaterialFunctionCall(UV=@n6, Size=@n10, Corner Radius=@n7, Skew Angle=@n8) {materialFunction=&MF_UI_CoreUI_RoundedMask}.Shape Mask
    B=MaterialFunctionCall(UV=@n12, Size=@n13, Corner Radius=@n7, Skew Angle=@n8) {materialFunction=&MF_UI_CoreUI_RoundedMask}.Shape Mask
   )
   B=Max(
    A=MaterialFunctionCall(UV=@n15, Size=@n10, Corner Radius=@n7, Skew Angle=@n8) {materialFunction=&MF_UI_CoreUI_RoundedMask}.Shape Mask
    B=MaterialFunctionCall(UV=@n16, Size=@n13, Corner Radius=@n7, Skew Angle=@n8) {materialFunction=&MF_UI_CoreUI_RoundedMask}.Shape Mask
   )
  )
  n18 = ScalarParameter() {parameterName=IsBottom}
  LinearInterpolate(A=@n17, B=Max(A=@n17, B=NamedRerouteUsage()), Alpha=@n18)
  n19 = VectorParameter() {parameterName=Color, defaultValue=<rgba 0.0648 0.0844 0.0395 1>}
  n20 = AppendVector(A=@n3, B=Multiply(A=@n4) {constB=0.15})
  n21 = Constant2Vector() {g=0.05, r=0.5}
  n22 = MaterialFunctionCall(UVs=@n20, CenterPosition=@n21, Radius=Constant() {r=0.55}, Density=Constant() {r=1.4}) {materialFunction=&RadialGradientExponential}
  n23 = MaterialFunctionCall(UVs=@n20, CenterPosition=@n21, Radius=Constant() {r=0.25}, Density=Constant() {r=3.3}) {materialFunction=&RadialGradientExponential}
  n24 = LinearInterpolate(
   A=LinearInterpolate(
    A=@n19.RGB
    B=VectorParameter() {parameterName=GlowColor, defaultValue=<rgba 0.1329 0.2195 0.1356 1>}.RGB
    Alpha=Max(A=@n22.RadialGradientExponential, B=@n23.RadialGradientExponential)
   )
   B=@n19.RGB
   Alpha=@n18
  )
  n25 = Multiply(A=@n24) {constB=2}
  LinearInterpolate(A=@n24, B=Saturate(None=Max(A=@n24, B=@n25)), Alpha=NamedRerouteUsage())
  n26 = Constant() {r=2}
  n27 = Constant() {r=2}
  n28 = Subtract(A=MaterialFunctionCall("Value (0-1)"=MaterialFunctionCall!(Animation Duration=@n26, Delay=@n27, Reset to Default) {materialFunction=/FortUILibrary/MaterialFunctions/Time/MF_UI_AnimationDelay}.Result) {materialFunction=&MF_UI_EaseCurves}.EaseOutCubic) {constB=0.5}
  NamedRerouteDeclaration(None=MaterialFunctionCall!(UVs, Offset=Multiply(A=@n28) {constB=4}, RotationCenter, RotationAngle=Constant()) {materialFunction=/FortUILibrary/MaterialFunctions/Gradients/MF_UI_ReflectedGradient}.Result) {name=Highlight, nodeColor=<rgba 0 0.8594 1 1>}
  n29 = NamedRerouteUsage()
  n30 = ScalarParameter() {defaultValue=0.03, parameterName=Angle}
  n31 = ScalarParameter() {defaultValue=0.01, parameterName=Blur}
  MaterialExpressionRemap_0 = MaterialXRemap!(None=@n29, InputLow=@n30, InputHigh=Subtract(A=@n30, B=@n31), TargetLow, TargetHigh) {inputHighDefault=0.027, inputLowDefault=0.03}
  n32 = ScalarParameter() {defaultValue=0.04, parameterName=Angle2}
  n33 = Subtract(A=@n32, B=@n31)
  n34 = MaterialXRemap(
   None=MaterialFunctionCall() {materialFunction=&LinearGradient}.VGradient
   InputHigh=Subtract(A=ScalarParameter() {defaultValue=0.2, parameterName=Step}, B=@n31)
  ) {inputHighDefault=0.199, inputLowDefault=0.2}
  NamedRerouteDeclaration(None=LinearInterpolate(
    A=@MaterialExpressionRemap_0
    B=MaterialXRemap(None=@n29, InputLow=@n32, InputHigh=@n33) {inputHighDefault=0.037, inputLowDefault=0.04}
    Alpha=Saturate(None=@n34)
   )) {name=CenterAlpha, nodeColor=<rgba 0.9688 1 0 1>}
  n35 = TextureCoordinate()
  n36 = ComponentMask(None=@n35) {g=true}
  n37 = ScalarParameter() {defaultValue=1, parameterName=RayCenter}
  n38 = MaterialFunctionCall!(Vector or UVs=AppendVector(A=ComponentMask(None=@n35) {r=true}, B=Subtract(A=@n36, B=@n37))) {materialFunction=/FortUILibrary/MaterialFunctions/Basics/MF_UI_CheaperVectorToRadial}
  n39 = Min(A=@n38.Vector Converted to Angle, B=OneMinus(None=@n38.Vector Converted to Angle))
  NamedRerouteDeclaration(None=Multiply(A=@n39) {constB=2}) {name=RaysShape, nodeColor=<rgba 0 1 0.4375 1>}
  OneMinus(None=@n39)
 ]
}
