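# Translucent UI material for the Athena HUD phase display, apparently a bus intro graphic, built from the T_UI_TrapezeBusIntro texture.
# The texture's red and green channels act as masks for the DarkerColor and LighterColor parameters, both dark blue by default, which are summed into one two-tone color.
# A separate Add doubles the red channel; the graph shows no links from either sum to material outputs.
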Material /Game/Athena/HUD/Phase/M_UI_TrapezeBusIntro
graph {
 properties {blendMode=BLEND_Translucent, materialDomain=MD_UI, pixelDepthOffsetMode=PDOM_Legacy, usedShadingModels=Default Lit}
 nodes [
  n1 = TextureSample() {texture=&/T_UI_TrapezeBusIntro}
  Add(
   A=Multiply(A=VectorParameter() {parameterName=DarkerColor, defaultValue=<rgba 0 0.0091 0.0823 1>}.RGB, B=@n1.R)
   B=Multiply(A=@n1.G, B=VectorParameter() {parameterName=LighterColor, defaultValue=<rgba 0 0.0242 0.2122 1>}.RGB)
  )
  Add(A=@n1.R, B=@n1.R)
 ]
}
