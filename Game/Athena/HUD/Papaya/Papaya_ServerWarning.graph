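# Papaya HUD server warning banner, derived from FortServerMessageWidget, showing an alert message with an animated countdown.
# OnAlertBegin sets the text, formats time as minutes or seconds, plays Intro and schedules Outro before collapsing.
# Starts over a minute get one NormalTick, while shorter ones run a one-second timer that updates seconds with FinalTick, adding sound in the last ten.
# CheckTagsForWidgetOffset shifts the banner using a playlist-tag-to-offset map.
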WidgetBlueprint /Game/Athena/HUD/Papaya/Papaya_ServerWarning
animations [
 NormalTick {
  binding_count=4
  display_label=""
  length_seconds=0.7834
  object=&self:NormalTick
  sequence {
   bindings [
    Image_RedBG {
     id=#1
     possessed_class=Image
     tracks [{
       type=MovieSceneFloatTrack
       display_name=Render Opacity
       sections [{type=MovieSceneFloatSection, channels [None {type=MovieSceneScriptingFloatChannel, keys [[3,0],[11,0.0952],[50.001,0]]}]}]
      }]
    }
    OverlayCountdownText {
     id=#2
     display_name="Image_RedBG (4)"
     possessed_class=TextBlock
     tracks [
      {
       type=MovieScene2DTransformTrack
       sections [{
         type=MovieScene2DTransformSection
         channels [
          Translation.X {type=MovieSceneScriptingFloatChannel}
          Translation.Y {type=MovieSceneScriptingFloatChannel}
          Angle {type=MovieSceneScriptingFloatChannel}
          Scale.X {type=MovieSceneScriptingFloatChannel, keys [[3,1.0016],[4,0.9016],[14,1.0267],[21,1.0016]]}
          Scale.Y {type=MovieSceneScriptingFloatChannel, keys [[3,1.0016],[4,0.9016],[14,1.0267],[21,1.0016]]}
          Shear.X {type=MovieSceneScriptingFloatChannel}
          Shear.Y {type=MovieSceneScriptingFloatChannel}
         ]
        }]
      }
      {
       type=MovieSceneFloatTrack
       display_name=Render Opacity
       sections [{type=MovieSceneFloatSection, channels [None {type=MovieSceneScriptingFloatChannel, const [1,2]}]}]
      }
     ]
    }
    OverlayCountdownNumber {
     id=#3
     possessed_class=Overlay
     tracks [{
       type=MovieSceneFloatTrack
       display_name=Render Opacity
       sections [{type=MovieSceneFloatSection, channels [None {type=MovieSceneScriptingFloatChannel, const [1,1]}]}]
      }]
    }
    Image_YellowBG {
     id=#4
     possessed_class=Image
     tracks [{
       type=MovieSceneFloatTrack
       display_name=Render Opacity
       sections [{type=MovieSceneFloatSection, channels [None {type=MovieSceneScriptingFloatChannel, keys [[3,0],[7,0.1184],[50.001,0]]}]}]
      }]
    }
   ]
   display_rate {numerator=60}
  }
 }
 FinalTick {
  binding_count=5
  display_label=""
  length_seconds=0.952
  object=&self:FinalTick
  sequence {bindings [
    Image_YellowBG {
     id=#5
     possessed_class=Image
     tracks [{
       type=MovieSceneFloatTrack
       display_name=Render Opacity
       sections [{
         type=MovieSceneFloatSection
         end_frame=19
         start_frame=0
         channels [None {type=MovieSceneScriptingFloatChannel, keys [[0,0.1624],[4,0.9232],[13,0]]}]
        }]
      }]
    }
    Image_Shade {
     id=#6
     possessed_class=Image
     tracks [{
       type=MovieSceneFloatTrack
       display_name=Render Opacity
       sections [{type=MovieSceneFloatSection, end_frame=19, start_frame=0, channels [None {type=MovieSceneScriptingFloatChannel, const [0,1]}]}]
      }]
    }
    OverlayCountdownNumber {
     id=#7
     possessed_class=Overlay
     tracks [{
       type=MovieSceneFloatTrack
       display_name=Render Opacity
       sections [{type=MovieSceneFloatSection, end_frame=19, start_frame=0, channels [None {type=MovieSceneScriptingFloatChannel, const [1,1]}]}]
      }]
    }
    TextBlock_Countdown {
     id=#8
     display_name="Image_YellowBG (5)"
     possessed_class=TextBlock
     tracks [
      {
       type=MovieScene2DTransformTrack
       sections [{
         type=MovieScene2DTransformSection
         end_frame=19
         start_frame=0
         channels [
          Translation.X {type=MovieSceneScriptingFloatChannel}
          Translation.Y {type=MovieSceneScriptingFloatChannel}
          Angle {type=MovieSceneScriptingFloatChannel}
          Scale.X {type=MovieSceneScriptingFloatChannel, keys [[0,0.667],[6.156,1.0376],[19,0.3],[19.039,1]]}
          Scale.Y {type=MovieSceneScriptingFloatChannel, keys [[0,0.667],[6.156,1.0376],[19,0.3]]}
          Shear.X {type=MovieSceneScriptingFloatChannel}
          Shear.Y {type=MovieSceneScriptingFloatChannel}
         ]
        }]
      }
      {
       type=MovieSceneFloatTrack
       display_name=Render Opacity
       sections [{type=MovieSceneFloatSection, end_frame=19, start_frame=0, channels [None {type=MovieSceneScriptingFloatChannel, keys [[1,1],[19.0003,0]]}]}]
      }
     ]
    }
    Image_RedBG {
     id=#9
     possessed_class=Image
     tracks [{
       type=MovieSceneFloatTrack
       display_name=Render Opacity
       sections [{type=MovieSceneFloatSection, end_frame=19, start_frame=0, channels [None {type=MovieSceneScriptingFloatChannel, keys [[0,1],[19.0003,0]]}]}]
      }]
    }
   ]}
 }
 FinalCountdown {
  binding_count=2
  display_label=""
  length_seconds=5.5283
  object=&self:FinalCountdown
  sequence {
   bindings [
    CommonTextBlock_0 {
     id=#10
     display_name="OverlayCountdownNumber (3)"
     possessed_class=CommonTextBlock
     tracks [
      {
       type=MovieScene2DTransformTrack
       sections [{
         type=MovieScene2DTransformSection
         end_frame=331
         start_frame=0
         channels [
          Translation.X {type=MovieSceneScriptingFloatChannel, const [0,1]}
          Translation.Y {type=MovieSceneScriptingFloatChannel, const [0,1]}
          Angle {type=MovieSceneScriptingFloatChannel, const [0,1]}
          Scale.X {type=MovieSceneScriptingFloatChannel, keys [[0,1],[9,0.3],[12,0.3445],[18,0.3123],[27,0.3289],[300.001,0.3],[306,0.3105],[315,0]]}
          Scale.Y {type=MovieSceneScriptingFloatChannel, keys [[0,1],[9,0.3],[12,0.3445],[18,0.3123],[27,0.3289],[300.001,0.3],[306,0.3105],[315,0]]}
          Shear.X {type=MovieSceneScriptingFloatChannel, const [0,1]}
          Shear.Y {type=MovieSceneScriptingFloatChannel, const [0,1]}
         ]
        }]
      }
      {
       type=MovieSceneFloatTrack
       display_name=Render Opacity
       sections [{
         type=MovieSceneFloatSection
         end_frame=331
         start_frame=0
         channels [None {type=MovieSceneScriptingFloatChannel, keys [[0,0],[4.252,1],[300.001,1],[322.701,0]]}]
        }]
      }
     ]
    }
    OverlayCountdownNumber {
     id=#11
     possessed_class=Overlay
     tracks [{
       type=MovieSceneFloatTrack
       display_name=Render Opacity
       sections [{type=MovieSceneFloatSection, end_frame=331, start_frame=0, channels [None {type=MovieSceneScriptingFloatChannel, const [1,1]}]}]
      }]
    }
   ]
   display_rate {numerator=60}
  }
 }
 FinalPhaseOutro {
  binding_count=1
  display_label=""
  length_seconds=0.25
  object=&self:FinalPhaseOutro
  sequence {
   bindings [OverlayCountdownNumber {
     id=#12
     possessed_class=Overlay
     tracks [{
       type=MovieSceneFloatTrack
       display_name=Render Opacity
       sections [{type=MovieSceneFloatSection, end_frame=30, start_frame=0, channels [None {type=MovieSceneScriptingFloatChannel, const [0,1]}]}]
      }]
    }]
   display_rate {numerator=120}
  }
 }
 Intro {
  binding_count=5
  length_seconds=2.1828
  object=&self:Intro
  sequence {bindings [
    Image_diamondpulse_1 {
     id=#13
     possessed_class=Image
     tracks [
      {
       type=MovieSceneWidgetMaterialTrack
       sections [{
         type=MovieSceneParameterSection
         channels [
          TimeOffset {type=MovieSceneScriptingFloatChannel, keys [[23.307,0],[43.6553,1]]}
          WidthStart {type=MovieSceneScriptingFloatChannel, const [0.2,1]}
          ManualTime {type=MovieSceneScriptingFloatChannel, const [1,1]}
          ScaleStart {type=MovieSceneScriptingFloatChannel, const [0,1]}
         ]
        }]
      }
      {
       type=MovieSceneFloatTrack
       display_name=Render Opacity
       sections [{type=MovieSceneFloatSection, channels [None {type=MovieSceneScriptingFloatChannel, keys [[23.307,0],[26.2657,1],[42.3107,1],[43.6553,0]]}]}]
      }
     ]
    }
    Overlay_2 {
     id=#14
     display_name="CommonBorder_0 (5)"
     possessed_class=Overlay
     tracks [{
       type=MovieSceneFloatTrack
       display_name=Render Opacity
       sections [{type=MovieSceneFloatSection, channels [None {type=MovieSceneScriptingFloatChannel, keys [[0,0],[6.779,1]]}]}]
      }]
    }
    Image_BG {
     id=#15
     possessed_class=Image
     tracks [{
       type=MovieSceneWidgetMaterialTrack
       sections [{type=MovieSceneParameterSection, channels [BorderMask {type=MovieSceneScriptingFloatChannel, keys [[0,1],[20.3483,0]]}]}]
      }]
    }
    Image_diamondpulse {
     id=#16
     possessed_class=Image
     tracks [
      {
       type=MovieSceneWidgetMaterialTrack
       sections [{
         type=MovieSceneParameterSection
         channels [
          TimeOffset {type=MovieSceneScriptingFloatChannel, keys [[20.3483,0],[40.6967,1]]}
          ManualTime {type=MovieSceneScriptingFloatChannel, const [1,1]}
          ScaleStart {type=MovieSceneScriptingFloatChannel, const [0,1]}
         ]
        }]
      }
      {
       type=MovieSceneFloatTrack
       display_name=Render Opacity
       sections [{type=MovieSceneFloatSection, channels [None {type=MovieSceneScriptingFloatChannel, keys [[20.3483,0],[22.1283,1],[39.25,1],[40.6967,0]]}]}]
      }
     ]
    }
    CommonBorder_0 {
     id=#17
     possessed_class=CommonBorder
     tracks [
      {
       type=MovieSceneFloatVectorTrack
       display_name=Desired Size Scale
       sections [{
         type=MovieSceneFloatVectorSection
         channels [
          Vector.X {type=MovieSceneScriptingFloatChannel, keys [[0,0],[14,1]]}
          Vector.Y {type=MovieSceneScriptingFloatChannel}
         ]
        }]
      }
      {
       type=MovieSceneByteTrack
       display_name=Clipping
       sections [{type=MovieSceneByteSection, channels [None {type=MovieSceneScriptingByteChannel, keys [[0,1],[14,0]]}]}]
      }
     ]
    }
   ]}
 }
 Outro {
  binding_count=5
  length_seconds=1.3921
  object=&self:Outro
  sequence {bindings [
    Image_diamondpulse_1 {
     id=#13
     possessed_class=Image
     tracks [
      {
       type=MovieSceneWidgetMaterialTrack
       sections [{
         type=MovieSceneParameterSection
         channels [
          TimeOffset {type=MovieSceneScriptingFloatChannel, keys [[0,1],[20.3483,0]]}
          WidthStart {type=MovieSceneScriptingFloatChannel, const [0.2,1]}
          ManualTime {type=MovieSceneScriptingFloatChannel, const [1,1]}
          ScaleStart {type=MovieSceneScriptingFloatChannel, const [0,1]}
         ]
        }]
      }
      {
       type=MovieSceneFloatTrack
       display_name=Render Opacity
       sections [{type=MovieSceneFloatSection, channels [None {type=MovieSceneScriptingFloatChannel, keys [[0,0],[1.3447,1],[17.3897,1],[20.3483,0]]}]}]
      }
     ]
    }
    Overlay_2 {
     id=#18
     possessed_class=Overlay
     tracks [{
       type=MovieSceneFloatTrack
       display_name=Render Opacity
       sections [{type=MovieSceneFloatSection, channels [None {type=MovieSceneScriptingFloatChannel, keys [[17.3897,1],[27.842,0]]}]}]
      }]
    }
    Image_BG {
     id=#15
     possessed_class=Image
     tracks [{
       type=MovieSceneWidgetMaterialTrack
       sections [{type=MovieSceneParameterSection, channels [BorderMask {type=MovieSceneScriptingFloatChannel, keys [[0,0],[10.656,1]]}]}]
      }]
    }
    Image_diamondpulse {
     id=#16
     possessed_class=Image
     tracks [
      {
       type=MovieSceneWidgetMaterialTrack
       sections [{
         type=MovieSceneParameterSection
         channels [
          TimeOffset {type=MovieSceneScriptingFloatChannel, keys [[2.9587,1],[23.307,0]]}
          ManualTime {type=MovieSceneScriptingFloatChannel, const [1,1]}
          ScaleStart {type=MovieSceneScriptingFloatChannel, const [0,1]}
         ]
        }]
      }
      {
       type=MovieSceneFloatTrack
       display_name=Render Opacity
       sections [{type=MovieSceneFloatSection, channels [None {type=MovieSceneScriptingFloatChannel, keys [[2.9587,0],[4.4053,1],[21.527,1],[23.307,0]]}]}]
      }
     ]
    }
    CommonBorder_0 {
     id=#17
     possessed_class=CommonBorder
     tracks [
      {
       type=MovieSceneFloatVectorTrack
       display_name=Desired Size Scale
       sections [{
         type=MovieSceneFloatVectorSection
         channels [
          Vector.X {type=MovieSceneScriptingFloatChannel, keys [[0,1],[27.842,0]]}
          Vector.Y {type=MovieSceneScriptingFloatChannel}
         ]
        }]
      }
      {
       type=MovieSceneByteTrack
       display_name=Clipping
       sections [{type=MovieSceneByteSection, end_frame=27, start_frame=0, channels [None {type=MovieSceneScriptingByteChannel, const [1,1]}]}]
      }
     ]
    }
   ]}
 }
]
graphs [
 CheckTagsForWidgetOffset {nodes [
   n1 = CheckTagsForWidgetOffset!() -> (then, GameplayTags:Array of Gameplay Tag Structures) {is=title K2Node_FunctionEntry}
   > n2 = For Each Loop with Break!(Exec:Exec, Array:Array of Gameplay Tag Structures=@n1, Break:Exec=@n4.then) -> (LoopBody:Exec, Array Element:Gameplay Tag Structure, Array Index:int, Completed:Exec)
   n3 = Find(
    TargetMap:Map of Gameplay Tag Structures to Vector 2D Structures=Get PlaylistOffsetMap() -> (PlaylistOffsetMap:Map of Gameplay Tag Structures to Vector 2D Structures)
    Key:"Gameplay Tag Structure (by ref)"=@n2
   ) -> (Value:Vector 2D Structure)
   > Branch(Condition=@n3.ReturnValue)
   > n4 = SetRenderTranslation(self=Get CountdownVerticalBox() -> (CountdownVerticalBox:Vertical Box Object Reference), Translation=@n3)
  ]}
 EventGraph {nodes [
   n1 = Event OnAlertBegin!() -> (OutputDelegate, then, SecondsRemainingStart:int=0, SecondsRemainingEnd:int=0, AlertText:"Text (by ref)") {is=title K2Node_Event}
   n2 = "SetText (Text)"(self=Get Text_Description() -> (Text_Description:Common Text Object Reference), InText=@n1.AlertText)
   n3 = Set StartSeconds(execute=@n1.then, StartSeconds:int=@n1.SecondsRemainingStart) -> (Output_Get:int=0)
   n4 = int / int!(A:int=@n3, B:int=60) -> (ReturnValue:int=0) {is=title K2Node_CallFunction}
   n5 = "SetText (Text)"(
    execute=[@n2.then, @n25.then]
    self=[
     Get TextBlock_Countdown_Shadow() -> (TextBlock_Countdown_Shadow:Common Text Object Reference)
     Get TextBlock_Countdown() -> (TextBlock_Countdown:Common Text Object Reference)
    ]
    InText=Select(
     Option 0:Text=Format Text!(Format="LOCTEXT(\"{0}s\")", 0:int=@n3) -> (Result)
     Option 1:Text=Format Text(Format="LOCTEXT(\"{0}m\")", 0=@n4)
     Index="integer > integer"(A=@n3, B=60) {is=title K2Node_CallFunction}
    ) -> (ReturnValue:Text)
   )
   n6 = Get Outro() -> (Outro:Widget Animation Object Reference)
   n7 = GetEndTime(self=@n6)
   n8 = int - int(A=@n3, B=@n1.SecondsRemainingEnd) {is=title K2Node_CallFunction}
   n9 = EventTimerEnd() {is=title K2Node_CustomEvent}
   n10 = Set Timer by Event(execute=@n5.then, Delegate=@n9, Time=float - float(A="To Float (Integer)"(InInt=@n8), B=@n7) {is=title K2Node_CallFunction}, bLooping=false)
   > Set TimerEnd(TimerEnd:Timer Handle Structure=@n10) -> (Output_Get:Timer Handle Structure)
   EventFinalCountdown() {is=title K2Node_CustomEvent}
   n11 = Get TimerEnd() -> (TimerEnd:Timer Handle Structure)
   n12 = int - int(
    A=Get StartSeconds() -> (StartSeconds:int=0)
    B=Truncate(A=Get Timer Elapsed Time by Handle!(Handle:Timer Handle Structure=@n11) -> (ReturnValue:float) {is=title K2Node_CallFunction})
   )
   > "SetText (Text)"(self=[Get TextBlock_Countdown_Shadow(), Get TextBlock_Countdown()], InText=Format Text(Format="LOCTEXT(\"{0}s\")", 0=@n12))
   > PlayAnimation(InAnimation=Get FinalTick() -> (FinalTick:Widget Animation Object Reference))
   > Decrement Int(Value=Get CurrentTimeRemaining() -> (CurrentTimeRemaining:int=0))
   > n13 = Sequence()
   > Branch(Condition="integer < integer"!(A:int=@n12, B:int=1) -> (ReturnValue:bool) {is=title K2Node_CallFunction})
   > Clear and Invalidate Timer by Handle(Handle=Get FinalCountdownLoop() -> (FinalCountdownLoop:Timer Handle Structure))
   Branch(execute=@n13.then_1, Condition="integer <= integer"!(A:int=@n12, B:int=10) -> (ReturnValue:bool) {is=title K2Node_CallFunction})
   > PlaySound2D(Sound=Get SoundOnTickFinal() -> (SoundOnTickFinal:Sound Base Object Reference), VolumeMultiplier=0.8)
   Event Construct()
   > "SetText (Text)"(self=[Get TextBlock_Countdown_Shadow(), Get TextBlock_Countdown()])
   > SetVisibility(InVisibility=Collapsed)
   Event PreConstruct()
   > SetVisibility(
    self=[
     Get Image_diamondpulse() -> (Image_diamondpulse:Image Object Reference)
     Get Image_diamondpulse_1() -> (Image_diamondpulse_1:Image Object Reference)
    ]
    InVisibility=Select(
     Option 0:ESlateVisibility Enum=HitTestInvisible
     Option 1:ESlateVisibility Enum=Collapsed
     Index=Get HideDiamondPulse() -> (HideDiamondPulse:bool)
    ) -> (ReturnValue:ESlateVisibility Enum)
   )
   > n14 = GetDynamicMaterial(self=Get Image_BG() -> (Image_BG:Image Object Reference))
   > SetVectorParameterValue(self=@n14, ParameterName=BorderColor, Value=Get BorderColor() -> (BorderColor:Linear Color Structure))
   SetVisibility(execute=@n22.then, InVisibility=Collapsed)
   PlaySound2D(execute=@n19.then, Sound=Get SoundOnTickNormal() -> (SoundOnTickNormal:Sound Base Object Reference))
   Set CurrentTimeRemaining(execute=@n21.then, CurrentTimeRemaining:int=Get StartSeconds() -> (StartSeconds=0)) -> (Output_Get:int=0)
   n15 = Get Intro()
   n16 = GetEndTime(self=@n15)
   n17 = "integer > integer"(A=Get StartSeconds() -> (StartSeconds=0), B=60)
   RetriggerableDelay(execute=@n23.then, Duration=SelectFloat(A=@n16, B=0, bPickA=@n17))
   > n18 = Branch(Condition=@n17)
   > n19 = PlayAnimation(InAnimation=Get NormalTick() -> (NormalTick:Widget Animation Object Reference))
   n20 = Set Timer by Event(execute=@n18.else, Delegate=Create Event(), Time=1, bLooping=true)
   > n21 = Set FinalCountdownLoop(FinalCountdownLoop:Timer Handle Structure=@n20) -> (Output_Get:Timer Handle Structure)
   PlayAnimation(execute=@n9.then, InAnimation=@n6)
   > RetriggerableDelay(Duration=@n7)
   > n22 = Clear and Invalidate Timer by Handle(Handle=Get FinalCountdownLoop())
   n23 = PlayAnimation(execute=@n24.then, InAnimation=@n15)
   n24 = StopAnimation(execute=@n26.then_1, InAnimation=@n15)
   n25 = SetVisibility(execute=@n26.then_0, InVisibility=HitTestInvisible)
   n26 = Sequence(execute=@n3.then)
  ]}
 PlayIntroAnimation {nodes [
   PlayIntroAnimation() {is=title K2Node_FunctionEntry}
   > PlayAnimationForward(InAnimation=Get Intro())
  ]}
 SetCountdownNumberVisibility {nodes [
   n1 = SetCountdownNumberVisibility!() -> (then, InVisibility:ESlateVisibility Enum=Visible) {is=title K2Node_FunctionEntry}
   > SetVisibility(self=Get OverlayCountdownNumber() -> (OverlayCountdownNumber:Overlay Object Reference), InVisibility=@n1)
  ]}
 SetCountdownText {nodes [
   n1 = SetCountdownText!() -> (then, InText:Text) {is=title K2Node_FunctionEntry}
   > "SetText (Text)"(self=[Get TextBlock_Countdown(), Get TextBlock_Countdown_Shadow()], InText=@n1)
  ]}
 SetTextDescription {nodes [
   n1 = SetTextDescription!() -> (then, InText:Text) {is=title K2Node_FunctionEntry}
   > "SetText (Text)"(self=Get Text_Description(), InText=@n1)
  ]}
 SetTextDescriptionJustification {nodes [
   n1 = SetTextDescriptionJustification!() -> (then, Justification:ETextJustify Enum=Left) {is=title K2Node_FunctionEntry}
   > Set Justification!(execute, Justification:ETextJustify Enum=@n1, self:Text Layout Widget Object Reference=Get Text_Description()) -> (then, Output_Get:ETextJustify Enum=Left)
  ]}
]
variables [
 bCountdownStarted {type=boolean}
 BorderColor {type=LinearColor, default=<rgba 0 1 0.3325 1>}
 CountdownLoopAudioEnabled {type=boolean}
 CurrentTimeRemaining {type=integer}
 CurveCountdownLoopPitch {type=CurveFloat, default=/Game/Athena/Sounds/LTMs/Score/C_CountdownLTM_TickLoop_Pitch}
 CurveCountdownLoopVolume {type=CurveFloat, default=/Game/Athena/Sounds/LTMs/Score/C_CountdownLTM_TickLoop_Volume}
 Final 60 Countdown Music Fadeout {type=number}
 Final 60 Countdown Music Instance {type=AudioComponent, default=null}
 FinalCountdownLoop {type=TimerHandle}
 FinalCountdownStartValue {type=integer, default=10}
 HideDiamondPulse {type=boolean}
 InterpolatedTimeRemaining {type=number}
 PlaylistOffsetMap {default {"(TagName=\"Athena.Playlist.CobaltRespawn\")"=<0 70>, "(TagName=\"Athena.Playlist.UraniumP\")"=<0 120>}}
 RelevantPlaylistTags {type=GameplayTagContainer, default {gameplayTags []}}
 SoundOnFinal10Countdown {type=SoundBase, default=/Game/Athena/Sounds/LTMs/Score/CountdownLTM_Final10Start_Cue}
 SoundOnFinal10Countdown_Music {type=SoundBase, default=null}
 SoundOnFinal60Countdown {type=SoundBase, default=/Game/Athena/Sounds/LTMs/Score/CountdownLTM_Final60Start_Cue}
 SoundOnFinal60Countdown_Music {type=SoundBase, default=null}
 SoundOnFinalCountdownWarning_Music {type=SoundBase, default=/Game/Athena/Sounds/LTMs/Score/ScoreLTM_FinalCountdown_Warning_Cue}
 SoundOnTickFinal {type=SoundBase, default=/Game/Athena/Sounds/LTMs/Score/CountdownLTM_FinalTick_Cue}
 SoundOnTickLoop {type=SoundBase, default=/Game/Athena/Sounds/LTMs/Score/CountdownLTM_TickLoop_Cue}
 SoundOnTickNormal {type=SoundBase, default=/Game/Athena/Sounds/LTMs/Score/CountdownLTM_NormalTick_Cue}
 StartSeconds {type=integer}
 TickLoopAudioComp {type=AudioComponent, default=null}
 TimerEnd {type=TimerHandle}
 VerticalOffset {type=Vector2D, default=<0 0>}
]
widgets {
 info {inheritedWidgetCount=1, parentClass=/Script/FortniteUI.FortServerMessageWidget, rootWidgetClass=&Overlay, widgetCount=19}
 widgets [
  Overlay_2 {widgetClassPath=&Overlay, widget=Overlay_2}
  CountdownVerticalBox {
   widgetClassPath=&VerticalBox
   bIsVariable=true
   parent=Overlay_2
   widget=CountdownVerticalBox
   slot {class=OverlaySlot, properties {horizontalAlignment=HAlign_Center}}
  }
  Overlay_0 {widgetClassPath=&Overlay, parent=CountdownVerticalBox, widget=Overlay_0, slot {class=VerticalBoxSlot}}
  Image_diamondpulse {
   widgetClassPath=&Image
   bIsVariable=true
   parent=Overlay_0
   widget=Image_diamondpulse
   properties {
    renderOpacity=0
    brush {
     bIsDynamicallyLoaded=false
     drawAs=Image
     imageType=NoImage
     mirroring=NoMirror
     resourceName=None
     resourceObject=/Game/Athena/HUD/LTM/Infiltration/Materials/M_UI_GameplayDiamondPing_Pulse_Add
     tiling=NoTile
     imageSize=<500 500>
     margin=<ltrb 0 0 0 0>
     outlineSettings {
      bUseBrushTransparency=false
      roundingType=HalfHeightRadius
      width=0
      color {colorUseRule=UseColor_Specified, specifiedColor=<rgba 0 0 0 0>}
      cornerRadii=<0 0 0 0>
     }
     tintColor {colorUseRule=UseColor_Specified, specifiedColor=<rgba 0 1 0.963 1>}
     uVRegion {bIsValid=false, max=<0 0>, min=<0 0>}
    }
   }
   slot {class=OverlaySlot, properties {horizontalAlignment=HAlign_Center, verticalAlignment=VAlign_Center, padding=<ltrb -250 -250 -250 -250>}}
  }
  Image_diamondpulse_1 {
   widgetClassPath=&Image
   bIsVariable=true
   parent=Overlay_0
   widget=Image_diamondpulse_1
   properties {
    renderOpacity=0
    brush {
     bIsDynamicallyLoaded=false
     drawAs=Image
     imageType=NoImage
     mirroring=NoMirror
     resourceName=None
     resourceObject=&M_UI_GameplayDiamondPing_Pulse_Add
     tiling=NoTile
     imageSize=<400 400>
     margin=<ltrb 0 0 0 0>
     outlineSettings {
      bUseBrushTransparency=false
      roundingType=HalfHeightRadius
      width=0
      color {colorUseRule=UseColor_Specified, specifiedColor=<rgba 0 0 0 0>}
      cornerRadii=<0 0 0 0>
     }
     tintColor {colorUseRule=UseColor_Specified, specifiedColor=<rgba 0 1 0.3601 1>}
     uVRegion {bIsValid=false, max=<0 0>, min=<0 0>}
    }
   }
   slot {class=OverlaySlot, properties {horizontalAlignment=HAlign_Center, verticalAlignment=VAlign_Center, padding=<ltrb -250 -250 -250 -250>}}
  }
  Image_BG {
   widgetClassPath=&Image
   bIsVariable=true
   parent=Overlay_0
   widget=Image_BG
   properties {brush {
     bIsDynamicallyLoaded=false
     drawAs=Box
     imageType=NoImage
     mirroring=NoMirror
     resourceName=None
     resourceObject=/Game/Athena/HUD/LTM/Papaya/Material/M_UI_SlantedBG_Inst
     tiling=NoTile
     imageSize=<128 80>
     margin=<ltrb 0.4 0 0.4 0>
     outlineSettings {
      bUseBrushTransparency=false
      roundingType=HalfHeightRadius
      width=0
      color {colorUseRule=UseColor_Specified, specifiedColor=<rgba 0 0 0 0>}
      cornerRadii=<0 0 0 0>
     }
     tintColor {colorUseRule=UseColor_Specified, specifiedColor=<rgba 1 1 1 1>}
     uVRegion {bIsValid=false, max=<0 0>, min=<0 0>}
    }}
   slot {class=OverlaySlot, properties {horizontalAlignment=HAlign_Fill, verticalAlignment=VAlign_Fill, padding=<ltrb -40 0 -40 0>}}
  }
  CommonBorder_0 {
   widgetClassPath=&CommonBorder
   parent=Overlay_0
   widget=CommonBorder_0
   properties {horizontalAlignment=HAlign_Center, verticalAlignment=VAlign_Center, padding=<ltrb 0 0 0 0>}
   slot {class=OverlaySlot, properties {horizontalAlignment=HAlign_Center, verticalAlignment=VAlign_Center}}
  }
  SizeBox_0 {
   widgetClassPath=&SizeBox
   parent=CommonBorder_0
   widget=SizeBox_0
   slot {class=BorderSlot, properties {horizontalAlignment=HAlign_Center, verticalAlignment=VAlign_Center, padding=<ltrb 0 0 0 0>}}
  }
  HorizontalBoxContainer {
   widgetClassPath=&HorizontalBox
   parent=SizeBox_0
   widget=HorizontalBoxContainer
   slot {class=SizeBoxSlot, properties {padding=<ltrb 30 0 30 0>}}
  }
  Text_Description {
   widgetClassPath=&CommonTextBlock
   bIsVariable=true
   parent=HorizontalBoxContainer
   widget=Text_Description
   properties {
    justification=Right
    lineHeightPercentage=0.8
    text="Party\r\nmoving in"
    textTransformPolicy=ToUpper
    font {
     bForceMonospaced=false
     bMaterialIsStencil=false
     fontMaterial=null
     fontObject=&BurbankSmall
     letterSpacing=0
     monospacedWidth=1
     size=18
     skewAmount=0
     typefaceFontName=Bold
     outlineSettings {
      bApplyOutlineToDropShadows=false
      bMiteredCorners=false
      bSeparateFillAlpha=false
      outlineMaterial=null
      outlineSize=0
      outlineColor=<rgba 0 0 0 1>
     }
    }
    shadowOffset=<0 0>
   }
   slot {class=HorizontalBoxSlot, properties {horizontalAlignment=HAlign_Right, verticalAlignment=VAlign_Center}}
  }
  OverlayCountdownNumber {
   widgetClassPath=&Overlay
   bIsVariable=true
   parent=HorizontalBoxContainer
   widget=OverlayCountdownNumber
   properties {renderTransform {angle=0, scale=<1 1>, shear=<0 0>, translation=<0 -5>}}
   slot {class=HorizontalBoxSlot, properties {horizontalAlignment=HAlign_Center, verticalAlignment=VAlign_Center, padding=<ltrb 18 5 0 5>}}
  }
  Overlay_1 {
   widgetClassPath=&Overlay
   parent=OverlayCountdownNumber
   widget=Overlay_1
   slot {class=OverlaySlot, properties {horizontalAlignment=HAlign_Center, verticalAlignment=VAlign_Center, padding=<ltrb -200 -200 -200 -200>}}
  }
  Image_Shade {
   widgetClassPath=&Image
   parent=Overlay_1
   widget=Image_Shade
   properties {
    renderOpacity=0
    brush {
     bIsDynamicallyLoaded=false
     drawAs=Image
     imageType=NoImage
     mirroring=NoMirror
     resourceName=None
     resourceObject=/Game/Athena/HUD/Materials/M_CountdownBG_Alpha_Inst
     tiling=NoTile
     imageSize=<482.0438 213.1235>
     margin=<ltrb 0 0 0 0>
     outlineSettings {
      bUseBrushTransparency=false
      roundingType=HalfHeightRadius
      width=0
      color {colorUseRule=UseColor_Specified, specifiedColor=<rgba 0 0 0 0>}
      cornerRadii=<0 0 0 0>
     }
     tintColor {colorUseRule=UseColor_Specified, specifiedColor=<rgba 0.0156 0.0063 0.0063 1>}
     uVRegion {bIsValid=false, max=<0 0>, min=<0 0>}
    }
    colorAndOpacity=<rgba 1 1 1 0.2286>
   }
   slot {class=OverlaySlot, properties {horizontalAlignment=HAlign_Center, verticalAlignment=VAlign_Center}}
  }
  Image_RedBG {
   widgetClassPath=&Image
   parent=Overlay_1
   widget=Image_RedBG
   properties {
    renderOpacity=0
    brush {
     bIsDynamicallyLoaded=false
     drawAs=Image
     imageType=NoImage
     mirroring=NoMirror
     resourceName=None
     resourceObject=/Game/Athena/HUD/Materials/M_CountdownBG_Base
     tiling=NoTile
     imageSize=<514.2557 300>
     margin=<ltrb 0 0 0 0>
     outlineSettings {
      bUseBrushTransparency=false
      roundingType=HalfHeightRadius
      width=0
      color {colorUseRule=UseColor_Specified, specifiedColor=<rgba 0 0 0 0>}
      cornerRadii=<0 0 0 0>
     }
     tintColor {colorUseRule=UseColor_Specified, specifiedColor=<rgba 1 1 1 1>}
     uVRegion {bIsValid=false, max=<0 0>, min=<0 0>}
    }
    colorAndOpacity=<rgba 0 1 0.8979 1>
   }
   slot {class=OverlaySlot, properties {horizontalAlignment=HAlign_Center, verticalAlignment=VAlign_Center}}
  }
  Image_YellowBG {
   widgetClassPath=&Image
   parent=Overlay_1
   widget=Image_YellowBG
   properties {
    renderOpacity=0
    brush {
     bIsDynamicallyLoaded=false
     drawAs=Image
     imageType=NoImage
     mirroring=NoMirror
     resourceName=None
     resourceObject=/Game/Athena/HUD/Materials/M_CountdownBG_Core
     tiling=NoTile
     imageSize=<462.1298 126.9874>
     margin=<ltrb 0 0 0 0>
     outlineSettings {
      bUseBrushTransparency=false
      roundingType=HalfHeightRadius
      width=0
      color {colorUseRule=UseColor_Specified, specifiedColor=<rgba 0 0 0 0>}
      cornerRadii=<0 0 0 0>
     }
     tintColor {colorUseRule=UseColor_Specified, specifiedColor=<rgba 1 1 1 1>}
     uVRegion {bIsValid=false, max=<0 0>, min=<0 0>}
    }
    colorAndOpacity=<rgba 0 1 0.2768 1>
   }
   slot {class=OverlaySlot, properties {horizontalAlignment=HAlign_Center, verticalAlignment=VAlign_Center}}
  }
  OverlayCountdownText {
   widgetClassPath=&Overlay
   parent=OverlayCountdownNumber
   widget=OverlayCountdownText
   properties {renderTransform {angle=0, scale=<1 1>, shear=<-7 -5>, translation=<0 0>}}
   slot {class=OverlaySlot}
  }
  TextBlock_Countdown_Shadow {
   widgetClassPath=&CommonTextBlock
   bIsVariable=true
   parent=OverlayCountdownText
   widget=TextBlock_Countdown_Shadow
   properties {
    text=60m
    colorAndOpacity {colorUseRule=UseColor_Specified, specifiedColor=<rgba 0 0 0 1>}
    font {
     bForceMonospaced=false
     bMaterialIsStencil=false
     fontMaterial=null
     fontObject=&BurbankSmall
     letterSpacing=0
     monospacedWidth=1
     size=56
     skewAmount=0
     typefaceFontName=Black
     outlineSettings {
      bApplyOutlineToDropShadows=false
      bMiteredCorners=false
      bSeparateFillAlpha=false
      outlineMaterial=null
      outlineSize=0
      outlineColor=<rgba 0 0 0 1>
     }
    }
    renderTransform {angle=0, scale=<1 1>, shear=<0 0>, translation=<5 5>}
    shadowColorAndOpacity=<rgba 0 0.8714 1 1>
    shadowOffset=<4 4>
   }
   slot {class=OverlaySlot}
  }
  TextBlock_Countdown {
   widgetClassPath=&CommonTextBlock
   bIsVariable=true
   parent=OverlayCountdownText
   widget=TextBlock_Countdown
   properties {
    text=60m
    font {
     bForceMonospaced=false
     bMaterialIsStencil=false
     fontMaterial=null
     fontObject=&BurbankSmall
     letterSpacing=0
     monospacedWidth=1
     size=56
     skewAmount=0
     typefaceFontName=Black
     outlineSettings {
      bApplyOutlineToDropShadows=false
      bMiteredCorners=false
      bSeparateFillAlpha=false
      outlineMaterial=null
      outlineSize=0
      outlineColor=<rgba 0 0 0 1>
     }
    }
    shadowOffset=<0 0>
   }
   slot {class=OverlaySlot}
  }
  TutorialHighlight {widgetClassPath=&TutorialHighlight}
 ]
}
