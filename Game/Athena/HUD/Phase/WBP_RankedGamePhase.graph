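# Ranked Battle Royale HUD banner for bus phases, showing rank icon, playlist name, phase message and countdown with bus or skydive icon.
# SetRankedProgress resolves tier display data with overrides, updates the icon, and RefreshColors tints text outline and dynamic materials from tier type colors.
# UpdateMessaging selects phase text, including Siphon notice; PlayIntroAnimations runs the intro, adding Competition Rules during GetReady when a ScalableFloat enables it.
# MaterialValidationCheck collapses particle images lacking valid dynamic materials.
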WidgetBlueprint /Game/Athena/HUD/Phase/WBP_RankedGamePhase
animations [
 Ranked_BusLockedIntro_Option2 {
  binding_count=7
  length_seconds=2.6504
  object=&self:Ranked_BusLockedIntro_Option2
  sequence {
   bindings [
    Image_Bus_Ranked {
     id=#1
     possessed_class=Image
     tracks [
      {
       type=MovieSceneWidgetMaterialTrack
       sections [{
         type=MovieSceneParameterSection
         channels [
          Edge {type=MovieSceneScriptingFloatChannel, keys [[10.9253,1],[16.0937,0.4393],[19.151,0.4476]]}
          ScaleAndOffsetR {type=MovieSceneScriptingFloatChannel, keys [[10.9253,0.52],[18.117,1.0337],[20.1947,1]]}
          ScaleAndOffsetG {type=MovieSceneScriptingFloatChannel, keys [[10.9253,0.58],[18.117,1.0364],[20.1947,1]]}
          ScaleAndOffsetB {type=MovieSceneScriptingFloatChannel, keys [[10.9253,0.03],[19.1513,0.5]]}
          ScaleAndOffsetA {type=MovieSceneScriptingFloatChannel, keys [[10.9253,0.97],[19.1513,0.5]]}
         ]
        }]
      }
      {
       type=MovieSceneFloatTrack
       display_name=Render Opacity
       sections [{type=MovieSceneFloatSection, channels [None {type=MovieSceneScriptingFloatChannel, keys [[-0.007,0],[0,1],[10.9253,0],[15.4367,1]]}]}]
      }
     ]
    }
    Text_BusTimer_Ranked {
     id=#2
     possessed_class=CommonTextBlock
     tracks [
      {
       type=MovieScene2DTransformTrack
       sections [{
         type=MovieScene2DTransformSection
         channels [
          Translation.X {type=MovieSceneScriptingFloatChannel}
          Translation.Y {type=MovieSceneScriptingFloatChannel}
          Angle {type=MovieSceneScriptingFloatChannel}
          Scale.X {type=MovieSceneScriptingFloatChannel, keys [[12.7647,2],[15.385,1]]}
          Scale.Y {type=MovieSceneScriptingFloatChannel, keys [[12.7647,2],[15.385,1]]}
          Shear.X {type=MovieSceneScriptingFloatChannel}
          Shear.Y {type=MovieSceneScriptingFloatChannel}
         ]
        }]
      }
      {
       type=MovieSceneFloatTrack
       display_name=Render Opacity
       sections [{type=MovieSceneFloatSection, channels [None {type=MovieSceneScriptingFloatChannel, keys [[0,0],[10.9253,0],[15.4367,1]]}]}]
      }
     ]
    }
    Overlay_TriangleParticlesOutward {
     id=#3
     possessed_class=Overlay
     tracks [
      {
       type=MovieSceneFloatTrack
       display_name=Render Opacity
       sections [{type=MovieSceneFloatSection, end_frame=44, start_frame=0, channels [None {type=MovieSceneScriptingFloatChannel, keys [[0,0],[30,0],[39,1]]}]}]
      }
      {
       type=MovieScene2DTransformTrack
       sections [{
         type=MovieScene2DTransformSection
         end_frame=47
         start_frame=0
         channels [
          Scale.X {type=MovieSceneScriptingFloatChannel, const [1,3]}
          Scale.Y {type=MovieSceneScriptingFloatChannel, const [1,3]}
         ]
        }]
      }
     ]
    }
    WBP_RankedGamePhase_C_139 {
     id=#4
     possessed_class=WBP_RankedGamePhase_C
     tracks [{
       type=MovieSceneFloatTrack
       display_name=Render Opacity
       sections [{type=MovieSceneFloatSection, channels [None {type=MovieSceneScriptingFloatChannel, keys [[0,0],[6.1947,0],[12.1947,1]]}]}]
      }]
    }
    Overlay_Habanero {
     id=#5
     possessed_class=Overlay
     tracks [
      {
       type=MovieSceneFloatTrack
       display_name=Render Opacity
       sections [{type=MovieSceneFloatSection, end_frame=27, start_frame=0, channels [None {type=MovieSceneScriptingFloatChannel, keys [[0,0],[20,0],[25,1]]}]}]
      }
      {
       type=MovieScene2DTransformTrack
       sections [{
         type=MovieScene2DTransformSection
         end_frame=56
         channels [
          Translation.Y {type=MovieSceneScriptingFloatChannel, keys [[-0.007,0],[32,0],[42,13]]}
          Scale.X {type=MovieSceneScriptingFloatChannel, keys [[-0.007,1],[20,4],[25,1],[42,1]]}
          Scale.Y {type=MovieSceneScriptingFloatChannel, keys [[0,1],[20,4],[25,1],[42,1]]}
         ]
        }]
      }
     ]
    }
    Image_Skydive_Ranked {
     id=#6
     possessed_class=Image
     tracks [
      {
       type=MovieSceneWidgetMaterialTrack
       sections [{
         type=MovieSceneParameterSection
         channels [
          Edge {type=MovieSceneScriptingFloatChannel, keys [[16.815,0.264],[17.9273,0.1949],[19.151,0.15]]}
          ScaleAndOffsetR {type=MovieSceneScriptingFloatChannel, keys [[10.9253,0.58],[18.117,1.0337],[20.1947,1]]}
          ScaleAndOffsetG {type=MovieSceneScriptingFloatChannel, keys [[10.9253,0.58],[18.117,1.0364],[20.1947,1]]}
          ScaleAndOffsetB {type=MovieSceneScriptingFloatChannel, keys [[10.9253,0.4],[19.1513,0.5]]}
          ScaleAndOffsetA {type=MovieSceneScriptingFloatChannel, keys [[10.9253,0.97],[19.1513,0.5]]}
         ]
        }]
      }
      {
       type=MovieSceneFloatTrack
       display_name=Render Opacity
       sections [{type=MovieSceneFloatSection, channels [None {type=MovieSceneScriptingFloatChannel, keys [[0,0],[10.9253,0],[15.4367,1]]}]}]
      }
     ]
    }
    Image_HabaneroBg {
     id=#7
     possessed_class=Image
     tracks [{
       type=MovieSceneFloatTrack
       display_name=Render Opacity
       sections [{type=MovieSceneFloatSection, end_frame=47, start_frame=0, channels [None {type=MovieSceneScriptingFloatChannel, keys [[0,0],[30,0],[39,1]]}]}]
      }]
    }
   ]
   tracks [{
     type=MovieSceneEventTrack
     sections [{
       type=MovieSceneEventTriggerSection
       channels [None {type=MovieSceneScriptingEventChannel, const ["<Struct 'MovieSceneEvent' (0x467aa842620) {}>", 1]}]
      }]
    }]
  }
 }
 Ranked_CompetitionTextIntro {
  binding_count=2
  length_seconds=3
  object=&self:Ranked_CompetitionTextIntro
  sequence {bindings [
    Text_BusPhase_Ranked {
     id=#8
     possessed_class=CommonTextBlock
     tracks [
      {
       type=MovieSceneFloatTrack
       display_name=Render Opacity
       sections [{type=MovieSceneFloatSection, end_frame=60, start_frame=0, channels [None {type=MovieSceneScriptingFloatChannel, keys [[0,0],[50,0],[60,1]]}]}]
      }
      {
       type=MovieSceneByteTrack
       display_name=Visibility
       sections [{type=MovieSceneByteSection, end_frame=0, start_frame=0, channels [None {type=MovieSceneScriptingByteChannel, const [0,1]}]}]
      }
     ]
    }
    Text_CompetitionRules {
     id=#9
     possessed_class=CommonTextBlock
     tracks [
      {
       type=MovieSceneByteTrack
       display_name=Visibility
       sections [{type=MovieSceneByteSection, end_frame=10, start_frame=0, channels [None {type=MovieSceneScriptingByteChannel, const [4,1]}]}]
      }
      {
       type=MovieScene2DTransformTrack
       sections [{
         type=MovieScene2DTransformSection
         end_frame=20
         start_frame=0
         channels [
          Scale.X {type=MovieSceneScriptingFloatChannel, keys [[10,3],[20,1]]}
          Scale.Y {type=MovieSceneScriptingFloatChannel, keys [[10,3],[20,1]]}
         ]
        }]
      }
      {
       type=MovieSceneColorTrack
       display_name=Color and Opacity
       sections [{
         type=MovieSceneColorSection
         end_frame=20
         start_frame=0
         channels [
          Color.R {type=MovieSceneScriptingFloatChannel, default=0.5776, const [1,2]}
          Color.G {type=MovieSceneScriptingFloatChannel, default=0.5776, keys [[10,1],[20,0.644]]}
          Color.B {type=MovieSceneScriptingFloatChannel, default=0.5776, keys [[0,0.018],[10,1],[20,0.018]]}
          Color.A {type=MovieSceneScriptingFloatChannel, const [1,2]}
         ]
        }]
      }
      {
       type=MovieSceneFloatTrack
       display_name=Render Opacity
       sections [{
         type=MovieSceneFloatSection
         end_frame=55
         start_frame=0
         channels [None {type=MovieSceneScriptingFloatChannel, keys [[0,0],[10,1],[50,1],[55,0]]}]
        }]
      }
     ]
    }
   ]}
 }
]
graphs [
 EventGraph {nodes [
   Delay(Duration=2)
   n1 = Get DebugTier() -> (DebugTier:int=0)
   > n2 = Branch(Condition="integer > integer"(A=@n1, B=17))
   > n3 = Set DebugTier(DebugTier:int=0) -> (Output_Get:int=0)
   n4 = Get BusWarning() -> (BusWarning:Common Visibility Switcher Object Reference)
   n5 = SetActiveWidget(execute=[@n9."  ", @n3.then], self=Get Switcher_AlertType() -> (Switcher_AlertType:Common Visibility Switcher Object Reference), Widget=@n4)
   SetActiveWidget(execute=[@n5.then, @n25.then], self=@n4, Widget=Get BusWarningRanked() -> (BusWarningRanked:Grid Panel Object Reference))
   n6 = Find(
    TargetMap:Map of Strings to Fort Habanero Mode List Structures=Get EntryData!(self:Fort Habanero Display Data Object Reference=Get RankData() -> (RankData:Fort Habanero Display Data Object Reference)) -> (EntryData:Map of Strings to Fort Habanero Mode List Structures)
    Key=ranked-br
   ) -> (Value:Fort Habanero Mode List Structure)
   n7 = Break Fort Habanero Tier!(FortHabaneroTier:"Fort Habanero Tier Structure (by ref)"="Get (a copy)"(
     Array:Array of Fort Habanero Tier Structures=Break Fort Habanero Mode List!(FortHabaneroModeList:"Fort Habanero Mode List Structure (by ref)"=@n6) -> (TierList:Array of Fort Habanero Tier Structures, bShowTopLeaderboardIcon:bool=~, TopTierTopLeaderboardPositions:Fort Habanero Tier Structure) {is=title K2Node_BreakStruct}
     Dimension 1=Get DebugTier() -> (DebugTier=0)
    ) -> (Output:Fort Habanero Tier Structure)) -> (
    ForegroundIcon:Material Instance Soft Object Reference
    ForegroundIconSmall:Material Instance Soft Object Reference
    BackgroundIcon:Material Instance Soft Object Reference
    BackgroundIconSmall:Material Instance Soft Object Reference
    WhiteIcon:Texture 2D Soft Object Reference
    SimpleIconSmall:Texture 2D Soft Object Reference
    Name:Text
    TierType:Fort Habanero Tier Type Data Object Reference
    AdditionalLeftIcon:Material Instance Soft Object Reference
    AdditionalRightIcon:Material Instance Soft Object Reference
   ) {is=title K2Node_BreakStruct}
   > UpdateDetails!(
    execute
    self:WBP Habanero Icon Object Reference=Get WBP_Habanero_Icon() -> (WBP_Habanero_Icon:WBP Habanero Icon Object Reference)
    ForegroundIcon:Material Instance Soft Object Reference=@n7.ForegroundIconSmall
    BackgroundIcon:Material Instance Soft Object Reference=@n7.BackgroundIconSmall
   ) -> (then)
   > SetVisibility(InVisibility=HitTestInvisible)
   > n8 = Cast To BP_UI_Habanero_TierTypeDataBase!(execute, Object=@n7.TierType) -> (then, CastFailed, AsBP UI Habanero Tier Type Data Base:BP UI Habanero Tier Type Data Base Object Reference) {is=title K2Node_DynamicCast}
   > RefreshColors!(execute, self, TierTypeData:BP UI Habanero Tier Type Data Base Object Reference=@n8) -> (then)
   n9 = Increment Int(" "=@n2.else, Value=@n1)
   n10 = Event SetPlaylistName!() -> (OutputDelegate, then, PlaylistName:Text) {is=title K2Node_Event}
   > n11 = Branch(Condition=TextIsEmpty(InText=@n10.PlaylistName))
   > SetVisibility(self=Get Text_BusPhase_PlaylistText() -> (Text_BusPhase_PlaylistText:Common Text Object Reference), InVisibility=Collapsed)
   n12 = Get Text_BusPhase_PlaylistText()
   "SetText (Text)"(execute=@n11.else, self=@n12, InText=@n10.PlaylistName)
   > SetVisibility(self=@n12, InVisibility=HitTestInvisible)
   n13 = Event PlayIntroAnimations!() -> (OutputDelegate, then, Step:EAthenaGamePhaseStep Enum=None) {is=title K2Node_Event}
   > SetVisibility(
    self=Get Overlay_TriangleParticlesOutward() -> (Overlay_TriangleParticlesOutward:Overlay Object Reference)
    InVisibility=SelfHitTestInvisible
   )
   > PlayAnimation(InAnimation=Get Ranked_BusLockedIntro_Option2() -> (Ranked_BusLockedIntro_Option2:Widget Animation Object Reference))
   n14 = Get CompetitionRulesEnabled() -> (CompetitionRulesEnabled:Scalable Float Structure)
   > Branch(Condition=AND Boolean(
     A="Equal (Enum)"(A:EAthenaGamePhaseStep Enum=@n13.Step, B:EAthenaGamePhaseStep Enum=GetReady)
     B=GetValueAsBool!(Input:"Scalable Float Structure (by ref)"=@n14, Level:float=0) -> (ReturnValue:bool)
    ))
   > PlayAnimation(InAnimation=Get Ranked_CompetitionTextIntro() -> (Ranked_CompetitionTextIntro:Widget Animation Object Reference))
   n15 = Event HandleGamePhaseStepChanged!() -> (OutputDelegate, then, Step:EAthenaGamePhaseStep Enum=None, ActiveIconIndex:int=0) {is=title K2Node_Event}
   > SetActiveWidgetIndex(
    self:Common Visibility Switcher Object Reference=Get Switcher_BusIconState_Ranked() -> (Switcher_BusIconState_Ranked:Common Visibility Switcher Object Reference)
    Index=@n15.ActiveIconIndex
   )
   Event HandleJumpAnimation() {is=title K2Node_Event}
   > StopAnimation(InAnimation=Get Ranked_BusLockedIntro_Option2())
   n16 = Event PreConstruct()
   > n17 = GetDynamicMaterial(self=Get Particles_Outward() -> (Particles_Outward:Image Object Reference))
   > SetVectorParameterValue(self=@n17, ParameterName=Color, Value=(R=0.0103,G=0.0103,B=0.0103,A=1))
   > MaterialValidationCheck()
   n18 = Event SetRankedProgress!() -> (OutputDelegate, then, ProgressKey:String, RankedTier:int=0, RankedPosition:int=0, RankedPercentage:double, RankedMMR:int=0) {is=title K2Node_Event}
   n19 = String To Name(InString=@n18.ProgressKey)
   n20 = Break Fort Habanero Tier(FortHabaneroTier=GetHabaneroDisplayDataOverride!(
     InDisplayData:Fort Habanero Display Data Object Reference=Get RankData()
     InOverrideData:Fort Compete UIHabanero Tier Override List Object Reference=Get Rank Info Overrides() -> (Rank Info Overrides:Fort Compete UIHabanero Tier Override List Object Reference)
     InKey:"Fort Compete UIHabanero Tier Override Key Structure (by ref)"=Make Fort Compete UIHabanero Tier Override Key!(
      RankedTrack:Name=@n19
      Season:String
      Rank:int=@n18.RankedTier
      MatchmakingRating:int=@n18.RankedMMR
      CurrentPosition:int=@n18.RankedPosition
     ) -> (FortCompeteUIHabaneroTierOverrideKey:Fort Compete UIHabanero Tier Override Key Structure) {is=title K2Node_MakeStruct}
    ) -> (OutData:Fort Habanero Tier Override Structure, ReturnValue:bool))
   > UpdateDetails(self=Get WBP_Habanero_Icon(), ForegroundIcon=@n20.ForegroundIconSmall, BackgroundIcon=@n20.BackgroundIconSmall)
   > SetVisibility(
    self=[
     Get Overlay_Habanero() -> (Overlay_Habanero:Overlay Object Reference)
     Get FortVisualAttachment_Triangles() -> (FortVisualAttachment_Triangles:Fort Visual Attachment Object Reference)
    ]
    InVisibility=HitTestInvisible
   )
   > n21 = Cast To BP_UI_Habanero_TierTypeDataBase(Object=@n20.TierType)
   > RefreshColors(TierTypeData=@n21)
   n22 = Event UpdateMessaging!() -> (OutputDelegate, then, Step:EAthenaGamePhaseStep Enum=None, Message:Text, TimeRemaining:Text, ShowSiphonText:bool) {is=title K2Node_Event}
   > "SetText (Text)"(self=Get Text_BusTimer_Ranked() -> (Text_BusTimer_Ranked:Common Text Object Reference), InText=@n22.TimeRemaining)
   > n23 = Switch on EAthenaGamePhaseStep!(execute, Selection:EAthenaGamePhaseStep Enum=@n22.Step) -> (
    None:Exec
    Setup:Exec
    Warmup:Exec
    GetReady:Exec
    BusLocked:Exec
    BusFlying:Exec
    StormForming:Exec
    StormHolding:Exec
    StormShrinking:Exec
    Countdown:Exec
    FinalCountdown:Exec
    EndGame:Exec
    Count:Exec
   ) {is=title K2Node_SwitchEnum}
   n24 = Get Text_BusPhase_Ranked() -> (Text_BusPhase_Ranked:Common Text Object Reference)
   "SetText (Text)"(execute=@n23.BusLocked, self=@n24, InText=Get BusPhaseLockedText() -> (BusPhaseLockedText:Text))
   "SetText (Text)"(
    execute=@n23.GetReady
    self=@n24
    InText=SelectText(
     A=Get SiphonOnText() -> (SiphonOnText:Text)
     B=Get BusPhaseGetReadyPhaseText() -> (BusPhaseGetReadyPhaseText:Text)
     bPickA=@n22.ShowSiphonText
    )
   )
   "SetText (Text)"(execute=@n23.BusFlying, self=@n24, InText=Get BusPhaseFlyingText() -> (BusPhaseFlyingText:Text))
   n25 = Branch(Condition=@n16.IsDesignTime)
   Select(Option 0=SelfHitTestInvisible, Option 1=Collapsed, Index="Equal (Enum)"(A=@n13.Step, B=GetReady))
  ]}
 MaterialValidationCheck {nodes [
   MaterialValidationCheck() {is=title K2Node_FunctionEntry}
   > Branch(Condition=IsUsingCompactUI())
   n1 = Get Particles_Outward()
   > n2 = GetDynamicMaterial(self=@n1)
   > n3 = Is Valid(InputObject=@n2)
   n4 = SetVisibility(execute=@n3.Is Not Valid, self=@n1, InVisibility=Collapsed)
   n5 = Get Particles_Outward_Small() -> (Particles_Outward_Small:Image Object Reference)
   n6 = GetDynamicMaterial(execute=[@n4.then, @n3.Is Valid], self=@n5)
   > n7 = Is Valid(InputObject=@n6)
   n8 = SetVisibility(execute=@n7.Is Not Valid, self=@n5, InVisibility=Collapsed)
   n9 = Get Particles_Outward_Medium() -> (Particles_Outward_Medium:Image Object Reference)
   n10 = GetDynamicMaterial(execute=[@n8.then, @n7.Is Valid], self=@n9)
   > n11 = Is Valid(InputObject=@n10)
   SetVisibility(execute=@n11.Is Not Valid, self=@n9, InVisibility=Collapsed)
  ]}
 RefreshColors {nodes [
   n1 = RefreshColors!() -> (then, TierTypeData:BP UI Habanero Tier Type Data Base Object Reference) {is=title K2Node_FunctionEntry}
   n2 = "To LinearColor (Color)"!(InColor:Color Structure=Get InnerBackground_Color!(self:BP UI Habanero Tier Type Data Base Object Reference=@n1) -> (InnerBackground_Color:Color Structure)) -> (ReturnValue:Linear Color Structure)
   > Set Outline Color!(In:Exec, CommonTextWidget:Common Text Object Reference=Get Text_BusTimer_Ranked(), Color:Linear Color Structure=@n2) -> (Out:Exec) {is=title K2Node_MacroInstance}
   > n3 = GetDynamicMaterial(self=Get Image_Bus_Ranked() -> (Image_Bus_Ranked:Image Object Reference))
   > n4 = GetDynamicMaterial(self=Get Image_Skydive_Ranked() -> (Image_Skydive_Ranked:Image Object Reference))
   > SetVectorParameterValue(self=[@n4, @n3], ParameterName=Outline Color, Value=@n2)
   > n5 = GetDynamicMaterial(self=Get Separator() -> (Separator:Image Object Reference))
   > n6 = GetDynamicMaterial(self=Get BackgroundBottom() -> (BackgroundBottom:Image Object Reference))
   > n7 = GetDynamicMaterial(self=Get BackgroundTop() -> (BackgroundTop:Image Object Reference))
   > SetVectorParameterValue(self=[@n7, @n6, @n5], ParameterName=Color, Value=@n2)
   > SetVectorParameterValue(
    self=@n7
    ParameterName=GlowColor
    Value="To LinearColor (Color)"(InColor=Get BarGlow_Color!(self:BP UI Habanero Tier Type Data Base Object Reference=@n1) -> (BarGlow_Color:Color Structure))
   )
   > n8 = GetDynamicMaterial(self=Get Particles_Outward_Small())
   > n9 = GetDynamicMaterial(self=Get Particles_Outward_Medium())
   > SetVectorParameterValue(
    self=[@n9, @n8]
    ParameterName=Color
    Value="To LinearColor (Color)"(InColor=Get ParticlesOutward_Color!(self:BP UI Habanero Tier Type Data Base Object Reference=@n1) -> (ParticlesOutward_Color:Color Structure))
   )
   Get ProgressBar_Color!(self:BP UI Habanero Tier Type Data Base Object Reference=@n1) -> (ProgressBar_Color:Color Structure)
  ]}
]
variables [
 BusPhaseFlyingText {type=string, default="Everybody Off!  Last stop in"}
 BusPhaseGetReadyPhaseText {type=string, default=Battle Bus launching in}
 BusPhaseLockedText {type=string, default=Doors will open in}
 BusWarning {type=CommonVisibilitySwitcher, default=null}
 CompetitionRulesDuration {type=number, default=4}
 CompetitionRulesEnabled {
  type=ScalableFloat
  default {
   registryType=None
   value=1
   curve {curveTable=/Game/Athena/Balance/DataTables/AthenaGameData, rowName=Default.UI.ShouldDisplayCompetitionRules}
  }
 }
 DebugTier {type=integer}
 Rank Info Overrides {type=FortCompeteUIHabaneroTierOverrideList, default=/Game/UI/Habanero/DA_RankDisplayDataOverride}
 RankData {type=FortHabaneroDisplayData, default=/Game/UI/Habanero/HabaneroDisplayData}
 ShowRanked {type=boolean}
 SiphonOnText {type=string, default="SIPHON IS ON!"}
]
widgets {
 info {inheritedWidgetCount=0, parentClass=&UserWidget, rootWidgetClass=&GridPanel, widgetCount=25}
 widgets [
  BusWarningRanked {widgetClassPath=&GridPanel, bIsVariable=true, widget=BusWarningRanked}
  Overlay_Habanero {
   widgetClassPath=&Overlay
   bIsVariable=true
   parent=BusWarningRanked
   widget=Overlay_Habanero
   properties {renderTransform {angle=0, scale=<1 1>, shear=<0 0>, translation=<0 13>}}
   slot {class=GridSlot, properties {horizontalAlignment=HAlign_Center, layer=2, verticalAlignment=VAlign_Top, padding=<ltrb 0 -25 0 0>}}
  }
  Image_HabaneroBg {
   widgetClassPath=&Image
   parent=Overlay_Habanero
   widget=Image_HabaneroBg
   properties {
    visibility=SelfHitTestInvisible
    brush {
     bIsDynamicallyLoaded=false
     drawAs=Image
     imageType=NoImage
     mirroring=NoMirror
     resourceName=None
     resourceObject=&/MI_UI_Habanero_Background_BRPhaseChange
     tiling=NoTile
     imageSize=<55 55>
     margin=<ltrb 0 0 0 0>
     outlineSettings {
      bUseBrushTransparency=false
      roundingType=HalfHeightRadius
      width=0
      color {colorUseRule=UseColor_Specified, specifiedColor=<rgba 0 0 0 0>}
      cornerRadii=<0 0 0 0>
     }
     tintColor {colorUseRule=UseColor_Specified, specifiedColor=<rgba 1 1 1 1>}
     uVRegion {bIsValid=false, max=<0 0>, min=<0 0>}
    }
    renderTransform {angle=0, scale=<1.25 1.25>, shear=<0 0>, translation=<0 3.5>}
   }
   slot {class=OverlaySlot, properties {horizontalAlignment=HAlign_Center, verticalAlignment=VAlign_Center}}
  }
  WBP_Habanero_Icon {
   widgetClassPath=/Game/UI/Frontend/Lobby/Habanero/WBP_Habanero_Icon
   bIsVariable=true
   parent=Overlay_Habanero
   widget=WBP_Habanero_Icon
   properties {
    debugBackground=/Game/UI/Frontend/Lobby/Habanero/Habanero_Materials/MaterialInstances/Icons/Background/T02/MI_UI_Habanero_Background_01
    debugForeground=/Game/UI/Frontend/Lobby/Habanero/Habanero_Materials/MaterialInstances/Icons/Foreground/T01/MI_UI_Habanero_Foreground_T01_01
    iconMinSize=<112 112>
    renderTransform {angle=0, scale=<1 1>, shear=<0 0>, translation=<0 3.5>}
   }
   slot {class=OverlaySlot, properties {horizontalAlignment=HAlign_Center}}
  }
  BackgroundTop {
   widgetClassPath=&Image
   bIsVariable=true
   parent=BusWarningRanked
   widget=BackgroundTop
   properties {brush {
     bIsDynamicallyLoaded=false
     drawAs=Image
     imageType=FullColor
     mirroring=NoMirror
     resourceName=None
     resourceObject=&/MI_UI_HUD_PhaseBG_Top
     tiling=NoTile
     imageSize=<500 14>
     margin=<ltrb 0 0 0 0>
     outlineSettings {
      bUseBrushTransparency=false
      roundingType=HalfHeightRadius
      width=0
      color {colorUseRule=UseColor_Specified, specifiedColor=<rgba 0 0 0 0>}
      cornerRadii=<0 0 0 0>
     }
     tintColor {colorUseRule=UseColor_Specified, specifiedColor=<rgba 1 1 1 1>}
     uVRegion {bIsValid=false, max=<0 0>, min=<0 0>}
    }}
   slot {class=GridSlot, properties {horizontalAlignment=HAlign_Center, row=1}}
  }
  BGC {
   widgetClassPath=&Image
   bIsVariable=true
   parent=BusWarningRanked
   widget=BGC
   properties {brush {
     bIsDynamicallyLoaded=false
     drawAs=Image
     imageType=FullColor
     mirroring=NoMirror
     resourceName=None
     resourceObject=&/MI_BackgroundGradient
     tiling=NoTile
     imageSize=<0 0>
     margin=<ltrb 0 0 0 0>
     outlineSettings {
      bUseBrushTransparency=false
      roundingType=HalfHeightRadius
      width=0
      color {colorUseRule=UseColor_Specified, specifiedColor=<rgba 0 0 0 0>}
      cornerRadii=<0 0 0 0>
     }
     tintColor {colorUseRule=UseColor_Specified, specifiedColor=<rgba 1 1 1 1>}
     uVRegion {bIsValid=false, max=<0 0>, min=<0 0>}
    }}
   slot {class=GridSlot, properties {layer=1, row=2, rowSpan=3, padding=<ltrb -80 0 -80 0>}}
  }
  Spacer_145 {
   widgetClassPath=&Spacer
   parent=BusWarningRanked
   widget=Spacer_145
   properties {size=<500 56>}
   slot {class=GridSlot, properties {row=2}}
  }
  Text_BusPhase_PlaylistText {
   widgetClassPath=&CommonTextBlock
   bIsVariable=true
   parent=BusWarningRanked
   widget=Text_BusPhase_PlaylistText
   properties {
    flowDirectionPreference=RightToLeft
    justification=Right
    text=Placeholder Playlist
    textTransformPolicy=ToUpper
    font {
     bForceMonospaced=false
     bMaterialIsStencil=false
     fontMaterial=null
     fontObject=&Burbank
     letterSpacing=0
     monospacedWidth=1
     size=30
     skewAmount=0.25
     typefaceFontName=BurbankRegular-Black
     outlineSettings {
      bApplyOutlineToDropShadows=false
      bMiteredCorners=false
      bSeparateFillAlpha=false
      outlineMaterial=null
      outlineSize=0
      outlineColor=<rgba 0 0 0 1>
     }
    }
    shadowOffset=<0 0>
   }
   slot {class=GridSlot, properties {horizontalAlignment=HAlign_Center, layer=2, row=2, verticalAlignment=VAlign_Center, padding=<ltrb 70 0 70 0>}}
  }
  Separator {
   widgetClassPath=&Image
   bIsVariable=true
   parent=BusWarningRanked
   widget=Separator
   properties {
    renderOpacity=0.7
    brush {
     bIsDynamicallyLoaded=false
     drawAs=Image
     imageType=FullColor
     mirroring=NoMirror
     resourceName=None
     resourceObject=&/MI_SeparatorGradient
     tiling=NoTile
     imageSize=<32 2>
     margin=<ltrb 0 0 0 0>
     outlineSettings {
      bUseBrushTransparency=false
      roundingType=HalfHeightRadius
      width=0
      color {colorUseRule=UseColor_Specified, specifiedColor=<rgba 0 0 0 0>}
      cornerRadii=<0 0 0 0>
     }
     tintColor {colorUseRule=UseColor_Specified, specifiedColor=<rgba 1 1 1 1>}
     uVRegion {bIsValid=false, max=<0 0>, min=<0 0>}
    }
   }
   slot {class=GridSlot, properties {layer=2, row=3}}
  }
  FortVisualAttachment_Triangles {
   widgetClassPath=&FortVisualAttachment
   bIsVariable=true
   parent=BusWarningRanked
   widget=FortVisualAttachment_Triangles
   properties {contentAnchor=<0.5 0.5>}
   slot {class=GridSlot, properties {horizontalAlignment=HAlign_Center, row=3, verticalAlignment=VAlign_Center}}
  }
  Overlay_TriangleParticlesOutward {
   widgetClassPath=&Overlay
   bIsVariable=true
   parent=FortVisualAttachment_Triangles
   widget=Overlay_TriangleParticlesOutward
   slot {class=SizeBoxSlot}
  }
  Particles_Outward {
   widgetClassPath=&Image
   bIsVariable=true
   parent=Overlay_TriangleParticlesOutward
   widget=Particles_Outward
   properties {
    visibility=SelfHitTestInvisible
    brush {
     bIsDynamicallyLoaded=false
     drawAs=Image
     imageType=FullColor
     mirroring=NoMirror
     resourceName=None
     resourceObject=/Game/UI/Frontend/Lobby/Habanero/Habanero_Materials/MaterialInstances/Background/MI_UI_Habanero_TrianglesEnergy_Outward
     tiling=NoTile
     imageSize=<1400 110>
     margin=<ltrb 0 0 0 0>
     outlineSettings {
      bUseBrushTransparency=false
      roundingType=HalfHeightRadius
      width=0
      color {colorUseRule=UseColor_Specified, specifiedColor=<rgba 0 0 0 0>}
      cornerRadii=<0 0 0 0>
     }
     tintColor {colorUseRule=UseColor_Specified, specifiedColor=<rgba 1 1 1 1>}
     uVRegion {bIsValid=false, max=<0 0>, min=<0 0>}
    }
   }
   slot {class=OverlaySlot, properties {horizontalAlignment=HAlign_Fill, verticalAlignment=VAlign_Fill}}
  }
  Particles_Outward_Medium {
   widgetClassPath=&Image
   bIsVariable=true
   parent=Overlay_TriangleParticlesOutward
   widget=Particles_Outward_Medium
   properties {
    visibility=SelfHitTestInvisible
    brush {
     bIsDynamicallyLoaded=false
     drawAs=Image
     imageType=FullColor
     mirroring=NoMirror
     resourceName=None
     resourceObject=/Game/UI/Frontend/Lobby/Habanero/Habanero_Materials/MaterialInstances/Background/MI_UI_Habanero_TrianglesEnergy_Outward_Medium
     tiling=NoTile
     imageSize=<512 43>
     margin=<ltrb 0 0 0 0>
     outlineSettings {
      bUseBrushTransparency=false
      roundingType=HalfHeightRadius
      width=0
      color {colorUseRule=UseColor_Specified, specifiedColor=<rgba 0 0 0 0>}
      cornerRadii=<0 0 0 0>
     }
     tintColor {colorUseRule=UseColor_Specified, specifiedColor=<rgba 1 1 1 1>}
     uVRegion {bIsValid=false, max=<0 0>, min=<0 0>}
    }
   }
   slot {class=OverlaySlot, properties {horizontalAlignment=HAlign_Fill, verticalAlignment=VAlign_Fill}}
  }
  Particles_Outward_Small {
   widgetClassPath=&Image
   bIsVariable=true
   parent=Overlay_TriangleParticlesOutward
   widget=Particles_Outward_Small
   properties {
    visibility=SelfHitTestInvisible
    brush {
     bIsDynamicallyLoaded=false
     drawAs=Image
     imageType=NoImage
     mirroring=NoMirror
     resourceName=None
     resourceObject=/Game/UI/Frontend/Lobby/Habanero/Habanero_Materials/MaterialInstances/Background/MI_UI_Habanero_TrianglesEnergy_Outward_Small
     tiling=NoTile
     imageSize=<512 43>
     margin=<ltrb 0 0 0 0>
     outlineSettings {
      bUseBrushTransparency=false
      roundingType=HalfHeightRadius
      width=0
      color {colorUseRule=UseColor_Specified, specifiedColor=<rgba 0 0 0 0>}
      cornerRadii=<0 0 0 0>
     }
     tintColor {colorUseRule=UseColor_Specified, specifiedColor=<rgba 1 1 1 1>}
     uVRegion {bIsValid=false, max=<0 0>, min=<0 0>}
    }
   }
   slot {class=OverlaySlot, properties {horizontalAlignment=HAlign_Fill, verticalAlignment=VAlign_Fill}}
  }
  Spacer {
   widgetClassPath=&Spacer
   parent=BusWarningRanked
   widget=Spacer
   properties {size=<0 46>}
   slot {class=GridSlot, properties {row=4}}
  }
  Text_BusPhase_Ranked {
   widgetClassPath=&CommonTextBlock
   bIsVariable=true
   parent=BusWarningRanked
   widget=Text_BusPhase_Ranked
   properties {
    flowDirectionPreference=RightToLeft
    justification=Right
    mobileFontSizeMultiplier=1.2
    text=This is placeholder text
    textTransformPolicy=ToUpper
    colorAndOpacity {colorUseRule=UseColor_Specified, specifiedColor=<rgba 0.5776 0.5776 0.5776 1>}
    font {
     bForceMonospaced=false
     bMaterialIsStencil=false
     fontMaterial=null
     fontObject=&BurbankSmall
     letterSpacing=0
     monospacedWidth=1
     size=15.75
     skewAmount=0
     typefaceFontName=Black
     outlineSettings {
      bApplyOutlineToDropShadows=false
      bMiteredCorners=false
      bSeparateFillAlpha=false
      outlineMaterial=null
      outlineSize=0
      outlineColor=<rgba 0 0 0 1>
     }
    }
    shadowOffset=<0 0>
   }
   slot {class=GridSlot, properties {horizontalAlignment=HAlign_Center, layer=2, row=4, verticalAlignment=VAlign_Center, padding=<ltrb 60 0 60 0>}}
  }
  Text_CompetitionRules {
   widgetClassPath=&CommonTextBlock
   bIsVariable=true
   parent=BusWarningRanked
   widget=Text_CompetitionRules
   properties {
    flowDirectionPreference=RightToLeft
    justification=Right
    mobileFontSizeMultiplier=1.2
    text=Competition Rules
    textTransformPolicy=ToUpper
    visibility=Hidden
    colorAndOpacity {colorUseRule=UseColor_Specified, specifiedColor=<rgba 1 0.6445 0.0185 1>}
    font {
     bForceMonospaced=false
     bMaterialIsStencil=false
     fontMaterial=null
     fontObject=&BurbankSmall
     letterSpacing=0
     monospacedWidth=1
     size=20.25
     skewAmount=0
     typefaceFontName=Black
     outlineSettings {
      bApplyOutlineToDropShadows=false
      bMiteredCorners=false
      bSeparateFillAlpha=false
      outlineMaterial=null
      outlineSize=1
      outlineColor=<rgba 0.0976 0.0648 0 1>
     }
    }
    shadowOffset=<0 0>
   }
   slot {class=GridSlot, properties {horizontalAlignment=HAlign_Center, layer=2, row=4, verticalAlignment=VAlign_Center, padding=<ltrb 60 0 60 0>}}
  }
  BackgroundBottom {
   widgetClassPath=&Image
   bIsVariable=true
   parent=BusWarningRanked
   widget=BackgroundBottom
   properties {brush {
     bIsDynamicallyLoaded=false
     drawAs=Image
     imageType=FullColor
     mirroring=NoMirror
     resourceName=None
     resourceObject=&/MI_UI_HUD_PhaseBG_Bottom
     tiling=NoTile
     imageSize=<416 48>
     margin=<ltrb 0 0 0 0>
     outlineSettings {
      bUseBrushTransparency=false
      roundingType=HalfHeightRadius
      width=0
      color {colorUseRule=UseColor_Specified, specifiedColor=<rgba 0 0 0 0>}
      cornerRadii=<0 0 0 0>
     }
     tintColor {colorUseRule=UseColor_Specified, specifiedColor=<rgba 1 1 1 1>}
     uVRegion {bIsValid=false, max=<0 0>, min=<0 0>}
    }}
   slot {class=GridSlot, properties {horizontalAlignment=HAlign_Center, row=5, verticalAlignment=VAlign_Top}}
  }
  CommonVisualAttachment_1 {
   widgetClassPath=&CommonVisualAttachment
   parent=BusWarningRanked
   widget=CommonVisualAttachment_1
   properties {contentAnchor=<1 0>, renderTransform {angle=0, scale=<1 1>, shear=<0 0>, translation=<-55 -5>}}
   slot {class=GridSlot, properties {horizontalAlignment=HAlign_Center, layer=2, row=5, verticalAlignment=VAlign_Top, padding=<ltrb 59 -7 0 0>}}
  }
  Overlay_BusIcon_Ranked {
   widgetClassPath=&Overlay
   parent=CommonVisualAttachment_1
   widget=Overlay_BusIcon_Ranked
   slot {class=SizeBoxSlot, properties {padding=<ltrb 0 0 -75 0>}}
  }
  HorizontalBox_2 {
   widgetClassPath=&HorizontalBox
   parent=Overlay_BusIcon_Ranked
   widget=HorizontalBox_2
   slot {class=OverlaySlot, properties {padding=<ltrb 0 10 0 0>}}
  }
  Text_BusTimer_Ranked {
   widgetClassPath=&CommonTextBlock
   bIsVariable=true
   parent=HorizontalBox_2
   widget=Text_BusTimer_Ranked
   properties {
    text=30s
    font {
     bForceMonospaced=false
     bMaterialIsStencil=false
     fontMaterial=null
     fontObject=&Burbank
     letterSpacing=0
     monospacedWidth=1
     size=31.5
     skewAmount=0
     typefaceFontName=BurbankRegular-Black
     outlineSettings {
      bApplyOutlineToDropShadows=false
      bMiteredCorners=false
      bSeparateFillAlpha=false
      outlineMaterial=null
      outlineSize=1
      outlineColor=<rgba 0.0742 0.007 0.007 1>
     }
    }
    renderTransform {angle=0, scale=<1 1>, shear=<-12 0>, translation=<0 0>}
    shadowOffset=<0 0>
   }
   slot {class=HorizontalBoxSlot, properties {horizontalAlignment=HAlign_Left, verticalAlignment=VAlign_Bottom}}
  }
  Switcher_BusIconState_Ranked {
   widgetClassPath=&CommonVisibilitySwitcher
   bIsVariable=true
   parent=HorizontalBox_2
   widget=Switcher_BusIconState_Ranked
   properties {renderTransform {angle=0, scale=<1 1>, shear=<0 0>, translation=<0 -2>}}
   slot {class=HorizontalBoxSlot}
  }
  Image_Bus_Ranked {
   widgetClassPath=&Image
   bIsVariable=true
   parent=Switcher_BusIconState_Ranked
   widget=Image_Bus_Ranked
   properties {
    visibility=SelfHitTestInvisible
    brush {
     bIsDynamicallyLoaded=false
     drawAs=Image
     imageType=FullColor
     mirroring=NoMirror
     resourceName=None
     resourceObject=&/MI_Bus_Icon
     tiling=NoTile
     imageSize=<56 56>
     margin=<ltrb 0 0 0 0>
     outlineSettings {
      bUseBrushTransparency=false
      roundingType=HalfHeightRadius
      width=0
      color {colorUseRule=UseColor_Specified, specifiedColor=<rgba 0 0 0 0>}
      cornerRadii=<0 0 0 0>
     }
     tintColor {colorUseRule=UseColor_Specified, specifiedColor=<rgba 1 1 1 1>}
     uVRegion {bIsValid=false, max=<0 0>, min=<0 0>}
    }
    renderTransform {angle=0, scale=<1 1>, shear=<0 0>, translation=<0 -1>}
   }
   slot {class=CommonVisibilitySwitcherSlot, properties {horizontalAlignment=HAlign_Center, verticalAlignment=VAlign_Center}}
  }
  Image_Skydive_Ranked {
   widgetClassPath=&Image
   bIsVariable=true
   parent=Switcher_BusIconState_Ranked
   widget=Image_Skydive_Ranked
   properties {
    visibility=SelfHitTestInvisible
    brush {
     bIsDynamicallyLoaded=false
     drawAs=Image
     imageType=FullColor
     mirroring=NoMirror
     resourceName=None
     resourceObject=&/MI_Skydive_Icon
     tiling=NoTile
     imageSize=<56 56>
     margin=<ltrb 0 0 0 0>
     outlineSettings {
      bUseBrushTransparency=false
      roundingType=HalfHeightRadius
      width=0
      color {colorUseRule=UseColor_Specified, specifiedColor=<rgba 0 0 0 0>}
      cornerRadii=<0 0 0 0>
     }
     tintColor {colorUseRule=UseColor_Specified, specifiedColor=<rgba 1 1 1 1>}
     uVRegion {bIsValid=false, max=<0 0>, min=<0 0>}
    }
    renderTransform {angle=0, scale=<1 1>, shear=<0 0>, translation=<0 3>}
   }
   slot {class=CommonVisibilitySwitcherSlot, properties {horizontalAlignment=HAlign_Center, verticalAlignment=VAlign_Center}}
  }
 ]
}
