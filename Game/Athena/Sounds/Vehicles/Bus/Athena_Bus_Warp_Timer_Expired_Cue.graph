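# Sound cue for the Athena Battle Bus that, going by its name, signals the warp countdown timer expiring.
# A single wave player plays the expired-timer wave directly, routed through the application tray notification sound class at boosted volume.
# Overridden concurrency keeps one instance at a time, stopping the oldest.
# Raised priority that bypasses volume scaling, a very large max distance and disabled virtualization are also set.
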SoundCue /Game/Athena/Sounds/Vehicles/Bus/Athena_Bus_Warp_Timer_Expired_Cue
root=WavePlayer_0
nodes [{id=WavePlayer_0, properties {soundWaveAssetPtr=&/Bus_Warp_Countdown_Timer_Expired_01}}]
properties {
 bBypassVolumeScaleForPriority=true
 bOverrideConcurrency=true
 maxDistance=2097152
 priority=1.2
 soundClassObject=/Game/Sounds/FortSoundClassesAndModes/Fort_Application_TrayNotification
 virtualizationMode=Disabled
 volumeMultiplier=1.75
 concurrencyOverrides {maxCount=1, resolutionRule=StopOldest}
}
