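# Sound cue for the Battle Bus doors, filed with Fortnite UI bus sounds, but its node graph is empty with no root, so it holds no audio.
# It is routed to the non-critical in-game UI sound class at slightly reduced volume.
# A concurrency override allows a single instance and prevents a new one from starting while it plays.
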SoundCue /Game/Sounds/Fort_UI_Sounds/Bus/Battle_Bus_Doors_Cue
root=null
nodes []
properties {
 bOverrideConcurrency=true
 maxDistance=2097152
 soundClassObject=&Fort_UI_InGame_NonCritical
 volumeMultiplier=0.85
 concurrencyOverrides {maxCount=1, resolutionRule=PreventNew}
}
