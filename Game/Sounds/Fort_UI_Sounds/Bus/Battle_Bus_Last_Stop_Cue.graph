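# Sound cue for the Battle Bus last-stop stinger, played as non-critical in-game UI audio.
# A single wave player feeds the S15 last-stop stinger through a modulator that adds slight random pitch variation at constant volume.
# Overridden concurrency allows one instance at a time and blocks new plays while it sounds.
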SoundCue /Game/Sounds/Fort_UI_Sounds/Bus/Battle_Bus_Last_Stop_Cue
root=Modulator_0
nodes [
 {id=Modulator_0, children [WavePlayer_0], properties {pitchMax=1.02, pitchMin=0.98, volumeMax=1, volumeMin=1}}
 {id=WavePlayer_0, properties {soundWaveAssetPtr=&/S15_Last_Stop_Stinger}}
]
properties {
 bOverrideConcurrency=true
 maxDistance=2097152
 soundClassObject=&Fort_UI_InGame_NonCritical
 volumeMultiplier=0.85
 concurrencyOverrides {maxCount=1, resolutionRule=PreventNew}
}
