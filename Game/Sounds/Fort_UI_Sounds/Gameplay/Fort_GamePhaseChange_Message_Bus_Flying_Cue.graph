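# Sound cue for the UI message sting that, judging by its name, accompanies the bus-flying game phase change.
# It delays a single info-update wave by a fixed interval and plays it quietly through the non-critical in-game UI sound class.
# Concurrency override allows only one instance and blocks new ones while it plays, and attenuation is overridden with its own falloff distance.
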SoundCue /Game/Sounds/Fort_UI_Sounds/Gameplay/Fort_GamePhaseChange_Message_Bus_Flying_Cue
root=Delay_0
nodes [
 {id=Delay_0, children [WavePlayer_24], properties {delayMax=0.75, delayMin=0.75}}
 {id=WavePlayer_24, properties {soundWaveAssetPtr=&/fort_UI_Info_Update}}
]
properties {
 bOverrideAttenuation=true
 bOverrideConcurrency=true
 maxDistance=2000
 soundClassObject=&Fort_UI_InGame_NonCritical
 volumeMultiplier=0.2
 attenuationOverrides {falloffDistance=1600}
 concurrencyOverrides {maxCount=1, resolutionRule=PreventNew}
}
