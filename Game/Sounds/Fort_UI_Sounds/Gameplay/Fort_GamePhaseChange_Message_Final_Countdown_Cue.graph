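# Sound cue named for the final countdown of a game phase change message, in the non-critical in-game UI sound class.
# It has no nodes and no root, so it produces no sound as authored.
# It still carries a reduced volume, an attenuation override and a concurrency override that allows one instance and prevents new ones while it plays.
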SoundCue /Game/Sounds/Fort_UI_Sounds/Gameplay/Fort_GamePhaseChange_Message_Final_Countdown_Cue
root=null
nodes []
properties {
 bOverrideAttenuation=true
 bOverrideConcurrency=true
 maxDistance=2000
 soundClassObject=&Fort_UI_InGame_NonCritical
 volumeMultiplier=0.2
 attenuationOverrides {falloffDistance=1600}
 concurrencyOverrides {maxCount=1, resolutionRule=PreventNew}
}
