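# Athena HUD widget that announces game phase changes, switching between a storm phase banner and a Battle Bus warning, with a separate server status display for Setup and Warmup.
# The BusWarning switcher hosts tag-slotted content, forwarding phase, message, playlist and ranked data through BPI_GamePhaseWidget.
# Intro animations are chosen by alert type and phase, with phase-specific sounds.
# Minigame state and spatial-gameplay checks adjust visibility and reset transforms.
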WidgetBlueprint /Game/Athena/HUD/Phase/AthenaGamePhaseChangeWidget
animations [
 ModeChanged {
  binding_count=1
  display_label=""
  length_seconds=6
  object=&self:ModeChanged
  sequence {bindings [AthenaGamePhaseChangeWidget_C_6 {
     id=#1
     possessed_class=AthenaGamePhaseChangeWidget_C
     tracks [
      {
       type=MovieScene2DTransformTrack
       sections [{
         type=MovieScene2DTransformSection
         channels [
          Translation.X {type=MovieSceneScriptingFloatChannel, keys [[0,700],[6,0],[114,0],[120,700]]}
          Translation.Y {type=MovieSceneScriptingFloatChannel, keys [[0,130],[6,0],[114,0],[120,130]]}
          Angle {type=MovieSceneScriptingFloatChannel}
          Scale.X {type=MovieSceneScriptingFloatChannel, keys [[0,0],[6,1],[114,1],[120,0]]}
          Scale.Y {type=MovieSceneScriptingFloatChannel, keys [[0,0],[6,1],[114,1],[120,0]]}
          Shear.X {type=MovieSceneScriptingFloatChannel}
          Shear.Y {type=MovieSceneScriptingFloatChannel}
         ]
        }]
      }
      {
       type=MovieSceneFloatTrack
       display_name=Render Opacity
       sections [{type=MovieSceneFloatSection, channels [None {type=MovieSceneScriptingFloatChannel, keys [[0,0],[6,1],[114,1],[120,0]]}]}]
      }
     ]
    }]}
 }
 IconPulse {
  binding_count=1
  display_label=""
  length_seconds=0.65
  object=&self:IconPulse
  sequence {
   bindings [Icon {
     id=#2
     possessed_class=Image
     tracks [{
       type=MovieScene2DTransformTrack
       sections [{
         type=MovieScene2DTransformSection
         end_frame=65
         start_frame=0
         channels [
          Translation.X {type=MovieSceneScriptingFloatChannel, const [0,4]}
          Translation.Y {type=MovieSceneScriptingFloatChannel, keys [[0,0],[39.9983,0],[44.9983,5],[55,0]]}
          Angle {type=MovieSceneScriptingFloatChannel}
          Scale.X {type=MovieSceneScriptingFloatChannel, keys [[0,1],[40,1],[44.9983,0.9],[55,1.5],[65,1]]}
          Scale.Y {type=MovieSceneScriptingFloatChannel, keys [[0,1],[40,1],[44.9983,0.9],[55,1.5],[65,1]]}
          Shear.X {type=MovieSceneScriptingFloatChannel}
          Shear.Y {type=MovieSceneScriptingFloatChannel}
         ]
        }]
      }]
    }]
   display_rate {numerator=100}
  }
 }
 IconSlide {
  binding_count=1
  display_label=""
  length_seconds=1.5
  object=&self:IconSlide
  sequence {
   bindings [Icon {
     id=#2
     possessed_class=Image
     tracks [{
       type=MovieScene2DTransformTrack
       sections [{
         type=MovieScene2DTransformSection
         end_frame=150
         start_frame=0
         channels [
          Translation.X {type=MovieSceneScriptingFloatChannel, keys [[0,-150],[44.9983,-150],[150,0]]}
          Translation.Y {type=MovieSceneScriptingFloatChannel, const [0,3]}
          Angle {type=MovieSceneScriptingFloatChannel}
          Scale.X {type=MovieSceneScriptingFloatChannel, keys [[35.9983,1],[38.9983,0.5],[50,1]]}
          Scale.Y {type=MovieSceneScriptingFloatChannel, const [1,3]}
          Shear.X {type=MovieSceneScriptingFloatChannel}
          Shear.Y {type=MovieSceneScriptingFloatChannel}
         ]
        }]
      }]
    }]
   display_rate {numerator=100}
  }
 }
 BusIntro {
  binding_count=7
  length_seconds=1.0956
  object=&self:BusIntro
  sequence {
   bindings [
    AthenaGamePhaseChangeWidget_C_124 {
     id=#3
     possessed_class=AthenaGamePhaseChangeWidget_C
     tracks [{
       type=MovieSceneFloatTrack
       display_name=Render Opacity
       sections [{type=MovieSceneFloatSection, channels [None {type=MovieSceneScriptingFloatChannel, keys [[0,0],[6.1947,0],[12.1947,1]]}]}]
      }]
    }
    Image_Bus {
     id=#4
     possessed_class=Image
     tracks [
      {
       type=MovieSceneWidgetMaterialTrack
       sections [{
         type=MovieSceneParameterSection
         channels [
          Edge {type=MovieSceneScriptingFloatChannel, keys [[10.9253,1],[16.0937,0.4393],[19.151,0.4476]]}
          ScaleAndOffsetR {type=MovieSceneScriptingFloatChannel, keys [[10.9253,0.52],[18.117,1.0337],[20.1947,1]]}
          ScaleAndOffsetG {type=MovieSceneScriptingFloatChannel, keys [[10.9253,0.58],[18.117,1.0364],[20.1947,1]]}
          ScaleAndOffsetB {type=MovieSceneScriptingFloatChannel, keys [[10.9253,0.03],[19.1513,0.5]]}
          ScaleAndOffsetA {type=MovieSceneScriptingFloatChannel, keys [[10.9253,0.97],[19.1513,0.5]]}
         ]
        }]
      }
      {
       type=MovieSceneFloatTrack
       display_name=Render Opacity
       sections [{type=MovieSceneFloatSection, channels [None {type=MovieSceneScriptingFloatChannel, keys [[10.9253,0],[15.4367,1]]}]}]
      }
     ]
    }
    Image_SkewBG {
     id=#5
     possessed_class=Image
     tracks [{
       type=MovieSceneWidgetMaterialTrack
       sections [{type=MovieSceneParameterSection, channels [ShineOffset {type=MovieSceneScriptingFloatChannel, keys [[10.013,0.2575],[21.9113,0.4857]]}]}]
      }]
    }
    HorizontalBox_0 {
     id=#6
     possessed_class=HorizontalBox
     tracks [{
       type=MovieSceneFloatTrack
       display_name=Render Opacity
       sections [{type=MovieSceneFloatSection, channels [None {type=MovieSceneScriptingFloatChannel, keys [[10.9253,0],[15.4367,1]]}]}]
      }]
    }
    BusWarning {
     id=#7
     possessed_class=CommonVisibilitySwitcher
     tracks [{
       type=MovieScene2DTransformTrack
       sections [{
         type=MovieScene2DTransformSection
         end_frame=11
         start_frame=0
         channels [
          Translation.X {type=MovieSceneScriptingFloatChannel}
          Translation.Y {type=MovieSceneScriptingFloatChannel}
          Angle {type=MovieSceneScriptingFloatChannel}
          Scale.X {type=MovieSceneScriptingFloatChannel, keys [[0,0],[6,0],[11,1]]}
          Scale.Y {type=MovieSceneScriptingFloatChannel, keys [[0,0],[6,0],[11,1]]}
          Shear.X {type=MovieSceneScriptingFloatChannel}
          Shear.Y {type=MovieSceneScriptingFloatChannel}
         ]
        }]
      }]
    }
    Text_BusTimer {
     id=#8
     possessed_class=CommonTextBlock
     tracks [{
       type=MovieScene2DTransformTrack
       sections [{
         type=MovieScene2DTransformSection
         channels [
          Translation.X {type=MovieSceneScriptingFloatChannel}
          Translation.Y {type=MovieSceneScriptingFloatChannel}
          Angle {type=MovieSceneScriptingFloatChannel}
          Scale.X {type=MovieSceneScriptingFloatChannel, keys [[12.7647,2],[15.385,1]]}
          Scale.Y {type=MovieSceneScriptingFloatChannel, keys [[12.7647,2],[15.385,1]]}
          Shear.X {type=MovieSceneScriptingFloatChannel}
          Shear.Y {type=MovieSceneScriptingFloatChannel}
         ]
        }]
      }]
    }
    Image_Skydive {
     id=#9
     possessed_class=Image
     tracks [{
       type=MovieSceneWidgetMaterialTrack
       sections [{
         type=MovieSceneParameterSection
         channels [
          Edge {type=MovieSceneScriptingFloatChannel, keys [[16.815,0.264],[17.9273,0.1949],[19.151,0.15]]}
          ScaleAndOffsetR {type=MovieSceneScriptingFloatChannel, keys [[10.9253,0.58],[18.117,1.0337],[20.1947,1]]}
          ScaleAndOffsetG {type=MovieSceneScriptingFloatChannel, keys [[10.9253,0.58],[18.117,1.0364],[20.1947,1]]}
          ScaleAndOffsetB {type=MovieSceneScriptingFloatChannel, keys [[10.9253,0.4],[19.1513,0.5]]}
          ScaleAndOffsetA {type=MovieSceneScriptingFloatChannel, keys [[10.9253,0.97],[19.1513,0.5]]}
         ]
        }]
      }]
    }
   ]
   tracks [{
     type=MovieSceneEventTrack
     sections [{
       type=MovieSceneEventTriggerSection
       channels [None {type=MovieSceneScriptingEventChannel, const ["<Struct 'MovieSceneEvent' (0x467aa840dc0) {}>", 1]}]
      }]
    }]
  }
 }
]
graphs [
 BP_HandleGamePhaseStepChanged {nodes [
   n1 = Handle Game Phase Step Changed!() -> (then, Step:EAthenaGamePhaseStep Enum=None) {is=title K2Node_FunctionEntry}
   > Parent: Handle Game Phase Step Changed!(execute, Step:EAthenaGamePhaseStep Enum=@n1) -> (then) {is=title K2Node_CallParentFunction}
   > n2 = Sequence()
   > n3 = For Each Loop(Array:Array of Widget Object References=GetAllChildren(self=Get BusWarning() -> (BusWarning:Common Visibility Switcher Object Reference))) -> (Array Element:Widget Object Reference)
   > n4 = Cast To BPI_GamePhaseWidget!(execute, Object=@n3) -> (then, CastFailed, AsBPI Game Phase Widget:BPI Game Phase Widget Interface) {is=title K2Node_DynamicCast}
   > HandleGamePhaseStepChanged!(
    execute
    self:BPI Game Phase Widget Interface=@n4
    Step:EAthenaGamePhaseStep Enum=@n1
    ActiveIconIndex:int=GetActiveWidgetIndex!(self:Common Visibility Switcher Object Reference=Get Switcher_BusIconState() -> (Switcher_BusIconState:Common Visibility Switcher Object Reference)) -> (ReturnValue:int=0)
   ) -> (then)
   Branch(
    execute=@n2.then_1
    Condition=OR Boolean(A="Equal (Enum)"(A:EAthenaGamePhaseStep Enum=@n1, B:EAthenaGamePhaseStep Enum=Setup), B="Equal (Enum)"(A=@n1, B=Warmup))
   )
   > SetVisibility(self=Get ServerStatus() -> (ServerStatus:Papaya Server Warning Object Reference), InVisibility=HitTestInvisible)
  ]}
 BP_ShouldShowSiphonText {nodes [
   Should Show Siphon Text() {is=title K2Node_FunctionEntry}
   n1 = GetValueAsBool!(
    Input:"Scalable Float Structure (by ref)"=Get CompetitionRulesEnabled!(self:WBP Ranked Game Phase Object Reference=Get WBP_RankedGamePhase() -> (WBP_RankedGamePhase:WBP Ranked Game Phase Object Reference)) -> (CompetitionRulesEnabled:Scalable Float Structure)
    Level:float=0
   ) -> (ReturnValue:bool)
   n2 = GetValueAsBool(Input=Get SiphonInstantHealthEnabled() -> (SiphonInstantHealthEnabled:Scalable Float Structure), Level=0)
   n3 = GetValueAsBool(Input=Get SiphonOverTimeEnabled() -> (SiphonOverTimeEnabled:Scalable Float Structure), Level=0)
   > Return Node!(
    execute
    ReturnValue:bool=AND Boolean!(A=OR Boolean(A=@n2, B=@n3), B=NOT Boolean(A=@n1), C:bool=Get ShowRanked() -> (ShowRanked:bool)) -> (ReturnValue)
   )
  ]}
 BP_UpdateMessaging {nodes [
   n1 = Update Messaging!() -> (then, Step:EAthenaGamePhaseStep Enum=None, NewMessage:"Text (by ref)", NewTimeText:"Text (by ref)") {is=title K2Node_FunctionEntry}
   > Parent: Update Messaging!(execute, Step:EAthenaGamePhaseStep Enum=@n1, NewMessage:"Text (by ref)"=@n1.NewMessage, NewTimeText:"Text (by ref)"=@n1.NewTimeText) -> (then) {is=title K2Node_CallParentFunction}
   > Set HasValidMessaging(HasValidMessaging:bool=true) -> (Output_Get:bool)
   > n2 = Sequence()
   > n3 = Switch on EAthenaGamePhaseStep!(execute, Selection:EAthenaGamePhaseStep Enum=Get Step() -> (Step:EAthenaGamePhaseStep Enum=None)) -> (
    None:Exec
    Setup:Exec
    Warmup:Exec
    GetReady:Exec
    BusLocked:Exec
    BusFlying:Exec
    StormForming:Exec
    StormHolding:Exec
    StormShrinking:Exec
    Countdown:Exec
    FinalCountdown:Exec
    EndGame:Exec
    Count:Exec
   ) {is=title K2Node_SwitchEnum}
   n4 = Get ServerStatus()
   SetTextDescription!(execute=[@n3.Setup, @n3.Warmup], self:Papaya Server Warning Object Reference=@n4, InText:Text="LOCTEXT(\"Waiting for\\r\\nPlayers\")") -> (then)
   n5 = Get Step() -> (Step=None)
   > SetTextDescriptionJustification!(
    execute
    self:Papaya Server Warning Object Reference=@n4
    Justification:ETextJustify Enum=Select(Option 0:ETextJustify Enum=Center, Option 1:ETextJustify Enum=Right, Index="Equal (Enum)"(A=@n5, B=Warmup)) -> (ReturnValue:ETextJustify Enum)
   ) -> (then)
   n6 = Get Step() -> (Step=None)
   > SetCountdownNumberVisibility!(
    execute
    self:Papaya Server Warning Object Reference=@n4
    InVisibility:ESlateVisibility Enum=Select(Option 0:ESlateVisibility Enum=Collapsed, Option 1:ESlateVisibility Enum=HitTestInvisible, Index="Equal (Enum)"(A=Warmup, B=@n6)) -> (ReturnValue:ESlateVisibility Enum)
   ) -> (then)
   > SetCountdownText!(execute, self:Papaya Server Warning Object Reference=@n4, InText:Text=Get NewTimeText() -> (NewTimeText:"Text (by ref)")) -> (then)
   n7 = For Each Loop(Exec=@n2.then_1, Array=GetAllChildren(self=Get BusWarning()))
   > n8 = Cast To BPI_GamePhaseWidget(Object=@n7)
   > UpdateMessaging!(
    execute
    self:BPI Game Phase Widget Interface=@n8
    Step:EAthenaGamePhaseStep Enum=Get Step() -> (Step=None)
    Message:Text=Get NewMessage() -> (NewMessage:"Text (by ref)")
    TimeRemaining:Text=Get NewTimeText()
    ShowSiphonText:bool=Get Show Siphon Text!(self:Athena HUDGame Phase Changing Base Object Reference) -> (ReturnValue:bool) {is=title K2Node_CallFunction}
   ) -> (then)
  ]}
 BindToMinigameEvents {nodes [
   n1 = BindToMinigameEvents!() -> (then, bShouldBind:bool=~) {is=title K2Node_FunctionEntry}
   > n2 = Get!(execute:Exec) -> (then:Exec, else:Exec, CachedMinigame:Fort Minigame Object Reference)
   n3 = GetPlayerSpawnComponent!(self:Fort Minigame Object Reference=@n2) -> (ReturnValue:Fort Minigame Player Spawn Component Object Reference)
   > Is Valid(InputObject=@n3)
   > n4 = Branch(Condition=@n1)
   n5 = Create Event()
   > Bind Event to On Player Spawn Override Devices Updated(self:Fort Minigame Player Spawn Component Object Reference=@n3, Delegate=@n5) {is=title K2Node_AddDelegate}
   Unbind Event from On Player Spawn Override Devices Updated(execute=@n4.else, self:Fort Minigame Player Spawn Component Object Reference=@n3, Delegate=@n5) {is=title K2Node_RemoveDelegate}
  ]}
 DoesMinigamePreventUpdatesAtPhase {nodes [
   n1 = DoesMinigamePreventUpdatesAtPhase!() -> (then, Value:Byte=0) {is=title K2Node_FunctionEntry}
   > n2 = Sequence()
   n3 = GetOwningPlayer()
   > n4 = Cast To FortPlayerControllerAthena(Object=@n3)
   > Is Valid(InputObject=GetMinigame!(ContextActor:Actor Object Reference=@n3) -> (ReturnValue:Fort Minigame Object Reference))
   n5 = MakeLiteralByte!(Value:Byte=Literal enum EAthenaGamePhaseStep!(Enum:EAthenaGamePhaseStep Enum=StormShrinking) -> (ReturnValue:EAthenaGamePhaseStep Enum) {is=title K2Node_EnumLiteral}) -> (ReturnValue:Byte=0)
   n6 = MakeLiteralByte(Value=Literal enum EAthenaGamePhaseStep(Enum=StormForming)) -> (ReturnValue=0)
   > Branch(Condition=AND Boolean!(
     A="Byte >= Byte"(A:Byte=@n1, B:Byte=@n6) -> (ReturnValue:bool=~) {is=title K2Node_PromotableOperator}
     B="Not Equal (Byte)"(A:Byte=@n5, B:Byte=@n1) -> (ReturnValue:bool=~) {is=title K2Node_PromotableOperator}
    ) -> (ReturnValue))
   n7 = MakeLiteralByte(Value=Literal enum EFortMinigameState!(Enum:EFortMinigameState Enum=InProgress) -> (ReturnValue:EFortMinigameState Enum) {is=title K2Node_EnumLiteral}) -> (ReturnValue=0)
   n8 = Get MinigameState() -> (MinigameState:EFortMinigameState Enum=PreGame)
   > Return Node(ReturnValue=OR Boolean(
     A=IsInAircraft!(self:Fort Player Controller Object Reference=@n4) -> (ReturnValue:bool)
     B="Byte < Byte"(A:Byte=@n8, B:Byte=@n7) -> (ReturnValue:bool=~) {is=title K2Node_PromotableOperator}
    ))
   Return Node(execute=@n2.then_1, ReturnValue=false)
  ]}
 EventGraph {nodes [
   EventHandleJump() {is=title K2Node_CustomEvent}
   > Branch(Condition="Equal (Enum)"(A=Get CurrentPhase() -> (CurrentPhase:EAthenaGamePhaseStep Enum=None), B=BusFlying))
   > n1 = Branch(Condition=Get ShowRanked())
   > n2 = Cast To BPI_GamePhaseWidget(Object=GetActiveWidget!(self:Common Visibility Switcher Object Reference=Get BusWarning()) -> (ReturnValue:Widget Object Reference))
   > n3 = HandleJumpAnimation(self:BPI Game Phase Widget Interface=@n2)
   SetRenderOpacity(execute=[@n4.then, @n3.then], InOpacity=0)
   n4 = StopAnimation(execute=@n1.else, InAnimation=Get BusIntro() -> (BusIntro:Widget Animation Object Reference))
   On AnimFinished() {is=title K2Node_CustomEvent}
   > SetHasActiveContent!(execute, self:Fort Prioritized Widget Interface, bNewActive:bool=false) -> (then)
   > SetVisibility(self=Get Switcher_AlertType() -> (Switcher_AlertType:Common Visibility Switcher Object Reference), InVisibility=Collapsed)
   > ForceResetWidgetIfSpatialGameplay()
   DevicesUpdated() {is=title K2Node_CustomEvent}
   > Branch(Condition="Equal (Enum)"(A:EFortMinigameState Enum=Get MinigameState() -> (MinigameState=PreGame), B:EFortMinigameState Enum=InProgress))
   > SetHasActiveContent(bNewActive=true)
   Event Destruct()
   > PopSoundMixModifier!(execute, InSoundMixModifier:Sound Mix Object Reference=Get LaunchingSoundMix() -> (LaunchingSoundMix:Sound Mix Object Reference)) -> (then)
   > BindToMinigameEvents!(execute, self, bShouldBind:bool=false) -> (then) {is=title K2Node_CallFunction}
   > UnregisterWidgetHost!(
    execute
    WidgetHost:Widget Object Reference
    WidgetHostTag:"Gameplay Tag Structure (by ref)"=Get BusWarningSlot() -> (BusWarningSlot:Gameplay Tag Structure)
   ) -> (then)
   n5 = Event MinigameStateChangedEvent!() -> (OutputDelegate, then, Minigame:Fort Minigame Object Reference, NewMinigameState:EFortMinigameState Enum=PreGame) {is=title K2Node_Event}
   > Parent: MinigameStateChangedEvent!(execute, Minigame:Fort Minigame Object Reference=@n5.Minigame, NewMinigameState:EFortMinigameState Enum=@n5.NewMinigameState) -> (then) {is=title K2Node_CallParentFunction}
   > n6 = Set MinigameState(MinigameState:EFortMinigameState Enum=@n5.NewMinigameState) -> (Output_Get:EFortMinigameState Enum=PreGame)
   > n7 = Switch on EFortMinigameState!(execute, Selection:EFortMinigameState Enum=@n6) -> (
    PreGame:Exec
    Setup:Exec
    Transitioning:Exec
    WaitingForCameras:Exec
    Warmup:Exec
    InProgress:Exec
    PostGameTimeDilation:Exec
    PostRoundEnd:Exec
    PostGameEnd:Exec
    PostGameAbandon:Exec
    PostGameReset:Exec
    Teardown:Exec
   ) {is=title K2Node_SwitchEnum}
   SetRenderOpacity(execute=@n7.Transitioning, InOpacity=0)
   StopAllAnimations(
    execute=[@n7.PostGameAbandon, @n7.PostGameEnd, @n7.PostGameReset, @n7.PostGameTimeDilation, @n7.PostRoundEnd]
    self:User Widget Object Reference
   )
   > ForceResetWidgetIfSpatialGameplay()
   n8 = Event Add Content To Host Widget!() -> (
    OutputDelegate
    then
    WidgetHostTag:"Gameplay Tag Structure (by ref)"
    ContentWidget:Widget Object Reference
    Padding:"Margin Structure (by ref)"
    bFill:bool
   ) {is=title K2Node_Event}
   > Branch(Condition="Equal (GameplayTag)"(A=@n8.WidgetHostTag, B=Get BusWarningSlot()))
   n9 = Get BusWarning()
   > AddChild(self=@n9, Content=@n8.ContentWidget)
   > SetActiveWidget(self=@n9, Widget=@n8.ContentWidget)
   Event Construct()
   > RegisterWidgetHost!(
    execute
    WidgetHost:Widget Object Reference
    WidgetHostTag:"Gameplay Tag Structure (by ref)"=Get BusWarningSlot()
    IsPlayerSpecific:bool=true
   ) -> (then)
   > n10 = PlaylistDataReadyAsync!(execute:Exec) -> (
    then:Exec
    OnReady:Exec
    GameState:Fort Game State Athena Object Reference
    Playlist:Fort Playlist Object Reference
    PlaylistContextTags:"Gameplay Tag Container Structure (by ref)"
   ) {is=title K2Node_AsyncAction}
   SetupForMinigame(execute=@n10.OnReady)
   > BindToAnimationFinished(Animation=Get ModeChanged() -> (ModeChanged:Widget Animation Object Reference), Delegate=Create Event())
   > n11 = Cast To FortPlayerControllerAthena(Object=GetOwningPlayer())
   > Assign On Aircraft State Change!(execute:Exec, self:Fort Player Controller Athena Object Reference=@n11, Delegate:"Delegate (by ref)"=Create Event()) -> (then:Exec) {is=title K2Node_AssignDelegate}
   > SetVisibility(self=Get Switcher_AlertType(), InVisibility=Collapsed)
   > Set HasValidMessaging(HasValidMessaging=false)
   > SetHasActiveContent(bNewActive=false)
   n12 = GetOwningPlayer()
   > Branch(Condition=AND Boolean(
     A=IsUsingCompactUI()
     B=IsUsingInputType!(InputType:ECommonInputType Enum=Touch, PlayerController:Player Controller Object Reference=@n12) -> (ReturnValue:bool)
    ))
   > SetPadding!(
    execute
    self:Vertical Box Slot Object Reference=SlotAsVerticalBoxSlot(Widget=Get Text_BusLine2() -> (Text_BusLine2:Common Text Object Reference))
    InPadding_Left:float=0
    InPadding_Top:float=2
    InPadding_Right:float=0
    InPadding_Bottom:float=0
   ) -> (then)
   > n13 = PlaylistDataReadyAsync()
   Branch(
    execute=@n13.OnReady
    Condition=HasTag!(
     TagContainer:"Gameplay Tag Container Structure (by ref)"=@n13.PlaylistContextTags
     Tag:Gameplay Tag Structure=(TagName="Athena.Playlist.TeamRumble")
     bExactMatch:bool=false
    ) -> (ReturnValue:bool)
   )
   > SetRenderTranslation!(execute, self=Get Overlay_BusWarning() -> (Overlay_BusWarning:Overlay Object Reference), Translation_X:double=0, Translation_Y:double=5) -> (then)
   n14 = Event BP_SetRankedProgress!() -> (OutputDelegate, then, ProgressKey:String, RankedTier:int=0, RankedPosition:int=0, RankedPercent:float, RankedMMR:int=0) {is=title K2Node_Event}
   > n15 = For Each Loop(Array=GetAllChildren(self=Get BusWarning()))
   > n16 = Cast To BPI_GamePhaseWidget(Object=@n15)
   > SetRankedProgress!(
    execute
    self:BPI Game Phase Widget Interface=@n16
    ProgressKey:String=@n14.ProgressKey
    RankedTier:int=@n14.RankedTier
    RankedPosition:int=@n14.RankedPosition
    RankedPercentage:double=@n14.RankedPercent
    RankedMMR:int=0
   ) -> (then)
   n17 = Event OnMinigameReady!() -> (OutputDelegate, then, Minigame:Fort Minigame Object Reference) {is=title K2Node_Event}
   > Parent: OnMinigameReady!(execute, Minigame:Fort Minigame Object Reference=@n17.Minigame) -> (then) {is=title K2Node_CallParentFunction}
   > Set CachedMinigame(CachedMinigame:Fort Minigame Object Reference=@n17.Minigame) -> (Output_Get:Fort Minigame Object Reference)
   > BindToMinigameEvents(bShouldBind=true)
   n18 = Event BP_SetPlaylistName!() -> (OutputDelegate, then, PlaylistName:"Text (by ref)") {is=title K2Node_Event}
   > n19 = For Each Loop(Array=GetAllChildren(self=Get BusWarning()))
   > n20 = Cast To BPI_GamePhaseWidget(Object=@n19)
   > SetPlaylistName!(execute, self:BPI Game Phase Widget Interface=@n20, PlaylistName:Text=@n18.PlaylistName) -> (then)
   n21 = Event BP_SetIsRanked!() -> (OutputDelegate, then, bIsRanked:bool) {is=title K2Node_Event}
   > Branch(Condition=GetShowRankedProgressCvar!(self:Athena HUDGame Phase Changing Base Object Reference) -> (ReturnValue:bool))
   > n22 = Set ShowRanked(ShowRanked:bool=@n21.bIsRanked) -> (Output_Get:bool)
   n23 = Get BusWarning()
   > SetActiveWidgetIndex(
    self:Common Visibility Switcher Object Reference=Get BusWarning()
    Index=SelectInt(A=1, B=GetActiveWidgetIndex(self=@n23) -> (ReturnValue=0), bPickA=@n22)
   )
  ]}
 ForceResetWidgetIfSpatialGameplay {nodes [
   ForceResetWidgetIfSpatialGameplay() {is=title K2Node_FunctionEntry}
   > Branch(Condition=IsSpatialGameplayExperience!() -> (ReturnValue:bool))
   > SetRenderTransform!(
    execute
    self:Widget Object Reference
    InTransform:Widget Transform Structure=Make Widget Transform!(Translation:Vector 2D Structure=(X=0,Y=0), Scale:Vector 2D Structure=(X=1,Y=1), Shear:Vector 2D Structure=(X=0,Y=0), Angle:float=0) -> (WidgetTransform:Widget Transform Structure) {is=title K2Node_MakeStruct}
   ) -> (then)
   > SetRenderOpacity(InOpacity=1)
  ]}
 HasSpawnOverrideComponent {nodes [
   HasSpawnOverrideComponent() {is=title K2Node_FunctionEntry}
   > n1 = GetSpawnOverrideComponent!(execute, self:Athena HUDGame Phase Changing Base Object Reference) -> (then, ReturnValue:Actor Component Object Reference)
   > n2 = Is Valid(InputObject=@n1)
   Return Node!(execute=@n2.Is Not Valid, bHasSpawnOverrideInterface:bool=false)
   > Return Node(bHasSpawnOverrideInterface=true)
  ]}
 IsRiftSpawning {nodes [
   IsRiftSpawning() {is=title K2Node_FunctionEntry}
   n1 = IsSpawningViaAircraft!() -> (bIsSpawningViaAircraft:bool)
   > Return Node!(
    execute
    ReturnValue:bool=AND Boolean!(
     A=IsSpatialGameplayExperience()
     B=HasSpawnOverrideComponent!() -> (bHasSpawnOverrideInterface:bool) {is=title K2Node_CallFunction}
     C:bool=NOT Boolean(A=@n1)
    ) -> (ReturnValue)
   )
  ]}
 IsSpawningViaAircraft {nodes [
   IsSpawningViaAircraft() {is=title K2Node_FunctionEntry}
   > n1 = Sequence!(execute) -> (then_0, then_1, then_2:Exec)
   > n2 = Cast To FortPlayerControllerAthena(Object=GetOwningPlayer())
   > Branch(Condition=IsInAircraft(self=@n2))
   > Return Node!(execute, bIsSpawningViaAircraft:bool=true)
   Return Node(execute=@n1.then_2, bIsSpawningViaAircraft=false)
   n3 = GetSpawnOverrideComponent(execute=@n1.then_1)
   > Is Valid(InputObject=@n3)
   > n4 = IsSpawningViaAircraft!(execute:Exec, self:Object Reference=@n3) -> (then:Exec, ReturnValue:bool) {is=title K2Node_Message}
   > Return Node(bIsSpawningViaAircraft=@n4)
  ]}
 PlayCountdownAudio {nodes [
   n1 = PlayCountdownAudio!() -> (then, NewTimeText:"Text (by ref)", Step:EAthenaGamePhaseStep Enum=None) {is=title K2Node_FunctionEntry}
   n2 = String To Integer(InString=LeftChop(SourceString="To String (Text)"(InText=@n1), Count=1))
   > Branch(Condition=AND Boolean!(A="integer <= integer"(A=@n2, B=10), B="Equal (Enum)"(A:EAthenaGamePhaseStep Enum=@n1.Step, B:EAthenaGamePhaseStep Enum=GetReady)) -> (ReturnValue))
   > n3 = Branch(Condition="integer < integer"(A=@n2, B=1))
   > n4 = PlaySound2D(Sound=/Game/Athena/Sounds/Vehicles/Bus/Athena_Bus_Warp_Timer_Expired_Cue)
   Return Node!(execute=[@n4.then, @n5.then])
   n5 = PlaySound2D(execute=@n3.else, Sound=/Game/Athena/Sounds/Vehicles/Bus/Athena_Bus_Warp_Timer_Tick_Default_Cue)
  ]}
 PlayGamePhaseAlertSound {nodes [
   n1 = PlayGamePhaseAlertSound!() -> (then, Step:EAthenaGamePhaseStep Enum=None) {is=title K2Node_FunctionEntry}
   n2 = Select!(
    None:Sound Base Object Reference
    Setup:Sound Base Object Reference=/Game/Sounds/Fort_UI_Sounds/Gameplay/Fort_GamePhaseChange_Message_Setup_Cue
    Warmup:Sound Base Object Reference=/Game/Sounds/Fort_UI_Sounds/Gameplay/Fort_GamePhaseChange_Message_Warmup_Cue
    GetReady:Sound Base Object Reference=/Game/Sounds/Fort_UI_Sounds/Gameplay/Fort_GamePhaseChange_Message_Get_Ready_Cue
    BusLocked:Sound Base Object Reference=/Game/Sounds/Fort_UI_Sounds/Gameplay/Fort_GamePhaseChange_Message_Bus_Locked_Cue
    BusFlying:Sound Base Object Reference=/Game/Sounds/Fort_UI_Sounds/Gameplay/Fort_GamePhaseChange_Message_Bus_Flying_Cue
    StormForming:Sound Base Object Reference=/Game/Sounds/Fort_UI_Sounds/Gameplay/Fort_GamePhaseChange_Message_Storm_Form_Cue
    StormHolding:Sound Base Object Reference=/Game/Sounds/Fort_UI_Sounds/Gameplay/Fort_GamePhaseChange_Message_Storm_Hold_Cue
    StormShrinking:Sound Base Object Reference=/Game/Sounds/Fort_UI_Sounds/Gameplay/Fort_GamePhaseChange_Message_Storm_Shrink_Cue
    Countdown:Sound Base Object Reference=/Game/Sounds/Fort_UI_Sounds/Gameplay/Fort_GamePhaseChange_Message_Countdown_Cue
    FinalCountdown:Sound Base Object Reference=/Game/Sounds/Fort_UI_Sounds/Gameplay/Fort_GamePhaseChange_Message_Final_Countdown_Cue
    EndGame:Sound Base Object Reference=/Game/Sounds/Fort_UI_Sounds/Gameplay/Fort_GamePhaseChange_Message_End_Game_Cue
    Count:Sound Base Object Reference=/Game/Sounds/Fort_UI_Sounds/Gameplay/Fort_GamePhaseChange_Message_Count_Cue
    Index:EAthenaGamePhaseStep Enum=@n1
   ) -> (ReturnValue:Sound Base Object Reference)
   > Is Valid(InputObject=@n2)
   > PlaySound2D(Sound=@n2)
  ]}
 PlayIntroAnimation {nodes [
   n1 = PlayIntroAnimation!() -> (then, Step:EAthenaGamePhaseStep Enum=None) {is=title K2Node_FunctionEntry}
   > Branch(Condition=Get HasValidMessaging() -> (HasValidMessaging:bool))
   > SetVisibility(self=Get Switcher_AlertType(), InVisibility=HitTestInvisible)
   > n2 = Branch(Condition=OR Boolean(A="Equal (Enum)"(A=@n1, B=Setup), B="Equal (Enum)"(A=@n1, B=Warmup)))
   > PlayIntroAnimation(self:Papaya Server Warning Object Reference=Get ServerStatus()) {is=title K2Node_CallFunction}
   n3 = Get Switcher_AlertType()
   n4 = Branch(execute=@n2.else, Condition="Equal (Integer)"(A=GetActiveWidgetIndex(self=@n3) -> (ReturnValue=0), B=0) {is=title K2Node_CallFunction})
   > PlayAnimation(InAnimation=Get ModeChanged())
   > PlayAnimation(InAnimation=Select!(
     Option 0:Widget Animation Object Reference=Get IconPulse() -> (IconPulse:Widget Animation Object Reference)
     Option 1:Widget Animation Object Reference=Get IconSlide() -> (IconSlide:Widget Animation Object Reference)
     Index:bool="Equal (Enum)"(A=@n1, B=StormShrinking)
    ) -> (ReturnValue:Widget Animation Object Reference))
   n5 = Get BusWarning()
   n6 = Branch(execute=@n4.else, Condition="Equal (Integer)"(A=GetActiveWidgetIndex(self=@n5) -> (ReturnValue=0)) {is=title K2Node_PromotableOperator})
   > PlayAnimation(InAnimation=Get BusIntro())
   n7 = Cast To BPI_GamePhaseWidget(execute=@n6.else, Object=GetActiveWidget(self=@n5))
   > PlayIntroAnimations!(execute, self:BPI Game Phase Widget Interface=@n7, Step:EAthenaGamePhaseStep Enum=@n1) -> (then)
  ]}
 RemoveContentFromWidgetHost {nodes [
   n1 = Remove Content From Host Widget!() -> (
    then
    WidgetHostTag:"Gameplay Tag Structure (by ref)"
    ContentWidget:Widget Object Reference
    RemovalType:EWidgetSlottingRemovalType Enum=Unparent
   ) {is=title K2Node_FunctionEntry}
   > n2 = Branch(Condition="Equal (GameplayTag)"(A=@n1, B=Get BusWarningSlot()))
   > n3 = RemoveChild(self=Get BusWarning(), Content=Get ContentWidget() -> (ContentWidget:Widget Object Reference))
   > n4 = Branch(Condition=@n3)
   n5 = Get BusWarning()
   n6 = "integer > integer"(A=GetChildrenCount!(self:Panel Widget Object Reference=@n5) -> (ReturnValue:int=0), B=2)
   > SetActiveWidgetIndex(self=@n5, Index=Select(Option 0:int=Select(Option 0=0, Option 1=2, Index=@n6), Option 1:int=1, Index=Get ShowRanked()) -> (ReturnValue:int))
   > Return Node!(execute, ReturnValue:bool=true)
   Return Node(execute=[@n4.else, @n2.else], ReturnValue=false)
  ]}
 Sequencer Events {nodes [
   BusIntro Audio() {is=title K2Node_CustomEvent}
   > PlaySound2D(Sound=Select!(
     None:Sound Base Object Reference
     Setup:Sound Base Object Reference
     Warmup:Sound Base Object Reference
     GetReady:Sound Base Object Reference=/Game/Sounds/Fort_UI_Sounds/Bus/Battle_Bus_Launching_Cue
     BusLocked:Sound Base Object Reference=/Game/Sounds/Fort_UI_Sounds/Bus/Battle_Bus_Doors_Cue
     BusFlying:Sound Base Object Reference=/Game/Sounds/Fort_UI_Sounds/Bus/Battle_Bus_Last_Stop_Cue
     StormForming:Sound Base Object Reference
     StormHolding:Sound Base Object Reference
     StormShrinking:Sound Base Object Reference
     Countdown:Sound Base Object Reference
     FinalCountdown:Sound Base Object Reference
     EndGame:Sound Base Object Reference
     Count:Sound Base Object Reference
     Index:EAthenaGamePhaseStep Enum=Get CurrentPhase() -> (CurrentPhase=None)
    ) -> (ReturnValue:Sound Base Object Reference))
  ]}
 SetupForMinigame {nodes [
   SetupForMinigame() {is=title K2Node_FunctionEntry}
   > Branch(Condition=IsSpatialGameplayExperience())
   > AppendHUDElementTags!(
    execute
    self:Fort HUDElement Widget Object Reference
    AdditionalTags:"Gameplay Tag Container Structure (by ref)"=MakeLiteralGameplayTagContainer!(Value:Gameplay Tag Container Structure=(GameplayTags=((TagName="Athena.Creative.HideForLoading")))) -> (ReturnValue:Gameplay Tag Container Structure)
   ) -> (then)
   n1 = Get ShouldDisplayGamePhaseUIOnStarterIsland() -> (ShouldDisplayGamePhaseUIOnStarterIsland:Scalable Float Structure)
   > Branch(Condition=NOT Boolean(A=GetValueAsBoolWithContext!(Input:"Scalable Float Structure (by ref)"=@n1, Level:float=0) -> (ReturnValue:bool)))
   n2 = Get JoiningLobbyTagContainer() -> (JoiningLobbyTagContainer:Gameplay Tag Container Structure)
   > AppendHUDElementTags(AdditionalTags=@n2)
   > AppendHUDElementTags(self=Get ServerStatus(), AdditionalTags=@n2)
  ]}
 ToggleCompass {nodes [
   n1 = ToggleCompass!() -> (then, bHide:bool) {is=title K2Node_FunctionEntry}
   > SetHUDElementVisibility!(
    execute
    self:Fort HUDContext Object Reference=Get FortHUDContext!() -> (ReturnValue:Fort HUDContext Object Reference) {is=title K2Node_GetBlueprintContextEx}
    HUDElementTags:"Gameplay Tag Container Structure (by ref)"=Get HideForSetupTags() -> (HideForSetupTags:Gameplay Tag Container Structure)
    bHideElements:bool=@n1
    ReasonName:"Name (by ref)"=None
   ) -> (then)
  ]}
]
variables [
 BusWarningSlot {type=GameplayTag, default {tagName=DynamicUI.HUD.Athena.GamePhase.BusWarning.Slot}}
 CachedMinigame {type=FortMinigame, default=null}
 HasValidMessaging {type=boolean}
 HideForSetupTags {type=GameplayTagContainer, default {gameplayTags [{tagName=HUD.Athena.Compass}]}}
 JoiningLobbyTagContainer {type=GameplayTagContainer, default {gameplayTags [{tagName=HUD.Hide.CreativeJoiningLobby}]}}
 MinigameState {type=EFortMinigameState, default=PreGame}
 ShouldDisplayGamePhaseUIOnStarterIsland {
  type=ScalableFloat
  default {registryType=Fort.DataRegistry, value=1, curve {curveTable=null, rowName=Default.UI.ShouldDisplayGamePhaseUIOnStarterIsland}}
 }
 ShowRanked {type=boolean}
 SiphonInstantHealthEnabled {
  type=ScalableFloat
  default {registryType=None, value=1, curve {curveTable=/Game/Athena/Balance/DataTables/AthenaGameData, rowName=Default.SD.Fun.InstantEnabled}}
 }
 SiphonOverTimeEnabled {
  type=ScalableFloat
  default {registryType=None, value=1, curve {curveTable=&AthenaGameData, rowName=Default.SD.Fun.OverTimeEnabled}}
 }
]
widgets {
 info {inheritedWidgetCount=9, parentClass=/Script/FortniteUI.AthenaHUDGamePhaseChangingBase, rootWidgetClass=&SizeBox, widgetCount=25}
 widgets [
  SizeBox_0 {
   widgetClassPath=&SizeBox
   widget=SizeBox_0
   properties {bOverride_MinDesiredHeight=true, bOverride_MinDesiredWidth=true, minDesiredHeight=80, minDesiredWidth=600}
  }
  Switcher_AlertType {
   widgetClassPath=&CommonVisibilitySwitcher
   bInherited=true
   bIsVariable=true
   parent=SizeBox_0
   widget=Switcher_AlertType
   slot {class=SizeBoxSlot, properties {horizontalAlignment=HAlign_Center, verticalAlignment=VAlign_Top}}
  }
  StormPhase {
   widgetClassPath=&Overlay
   bIsVariable=true
   parent=Switcher_AlertType
   widget=StormPhase
   slot {class=CommonVisibilitySwitcherSlot}
  }
  Image_BG {
   widgetClassPath=&Border
   parent=StormPhase
   widget=Image_BG
   properties {
    background {
     bIsDynamicallyLoaded=false
     drawAs=Box
     imageType=NoImage
     mirroring=NoMirror
     resourceName=None
     resourceObject=/Game/Athena/HUD/Art/Materials/M_UI_HUD_GamePhase_Backing
     tiling=NoTile
     imageSize=<72 72>
     margin=<ltrb 0.5 0.5 0.5 0.5>
     outlineSettings {
      bUseBrushTransparency=false
      roundingType=HalfHeightRadius
      width=0
      color {colorUseRule=UseColor_Specified, specifiedColor=<rgba 0 0 0 0>}
      cornerRadii=<0 0 0 0>
     }
     tintColor {colorUseRule=UseColor_Specified, specifiedColor=<rgba 1 1 1 1>}
     uVRegion {bIsValid=false, max=<0 0>, min=<0 0>}
    }
    padding=<ltrb 8 8 24 8>
   }
   slot {class=OverlaySlot, properties {horizontalAlignment=HAlign_Fill, verticalAlignment=VAlign_Fill}}
  }
  HorizontalBox_1 {
   widgetClassPath=&HorizontalBox
   parent=Image_BG
   widget=HorizontalBox_1
   slot {class=BorderSlot, properties {padding=<ltrb 8 8 24 8>}}
  }
  Icon {
   widgetClassPath=&Image
   bInherited=true
   bIsVariable=true
   parent=HorizontalBox_1
   widget=Icon
   properties {
    brush {
     bIsDynamicallyLoaded=false
     drawAs=Image
     imageType=NoImage
     mirroring=NoMirror
     resourceName=None
     resourceObject=/Game/Athena/HUD/Materials/M_UI_Timer_L_Filled
     tiling=NoTile
     imageSize=<56 56>
     margin=<ltrb 0 0 0 0>
     outlineSettings {
      bUseBrushTransparency=false
      roundingType=HalfHeightRadius
      width=0
      color {colorUseRule=UseColor_Specified, specifiedColor=<rgba 0 0 0 0>}
      cornerRadii=<0 0 0 0>
     }
     tintColor {colorUseRule=UseColor_Specified, specifiedColor=<rgba 1 1 1 1>}
     uVRegion {bIsValid=false, max=<0 0>, min=<0 0>}
    }
    renderTransformPivot=<0 1>
   }
   slot {class=HorizontalBoxSlot, properties {horizontalAlignment=HAlign_Left, verticalAlignment=VAlign_Center}}
  }
  Message {
   widgetClassPath=&CommonTextBlock
   bInherited=true
   bIsVariable=true
   parent=HorizontalBox_1
   widget=Message
   properties {
    text=STORM EYE SHRINKS IN
    textTransformPolicy=ToUpper
    font {
     bForceMonospaced=false
     bMaterialIsStencil=false
     fontMaterial=null
     fontObject=&BurbankSmall
     letterSpacing=0
     monospacedWidth=1
     size=16
     skewAmount=0
     typefaceFontName=Black
     outlineSettings {
      bApplyOutlineToDropShadows=false
      bMiteredCorners=false
      bSeparateFillAlpha=true
      outlineMaterial=&M-FontShadow
      outlineSize=3
      outlineColor=<rgba 0 0 0 1>
     }
    }
    renderTransform {angle=0, scale=<1 1>, shear=<-12 0>, translation=<0 -4>}
    shadowOffset=<0 0>
   }
   slot {class=HorizontalBoxSlot, properties {verticalAlignment=VAlign_Center, padding=<ltrb 8 0 0 0>}}
  }
  TimeText {
   widgetClassPath=&CommonTextBlock
   bInherited=true
   bIsVariable=true
   parent=StormPhase
   widget=TimeText
   properties {
    text=3m 20s
    font {
     bForceMonospaced=false
     bMaterialIsStencil=false
     fontMaterial=null
     fontObject=&Burbank
     letterSpacing=0
     monospacedWidth=1
     size=26
     skewAmount=0
     typefaceFontName=BurbankRegular-Black
     outlineSettings {
      bApplyOutlineToDropShadows=true
      bMiteredCorners=false
      bSeparateFillAlpha=true
      outlineMaterial=null
      outlineSize=2
      outlineColor=<rgba 0 0 0 0.75>
     }
    }
    renderTransform {angle=0, scale=<1 1>, shear=<-12 0>, translation=<0 18>}
    shadowColorAndOpacity=<rgba 0 0 0 0.666>
    shadowOffset=<4 4>
   }
   slot {class=OverlaySlot, properties {horizontalAlignment=HAlign_Center, verticalAlignment=VAlign_Bottom, padding=<ltrb 8 0 0 0>}}
  }
  BusWarning {
   widgetClassPath=&CommonVisibilitySwitcher
   bIsVariable=true
   parent=Switcher_AlertType
   widget=BusWarning
   slot {class=CommonVisibilitySwitcherSlot}
  }
  SizeBox_Default {
   widgetClassPath=&SizeBox
   parent=BusWarning
   widget=SizeBox_Default
   properties {bOverride_HeightOverride=true, heightOverride=71, renderTransform {angle=0, scale=<1 1>, shear=<0 0>, translation=<15 0>}}
   slot {class=CommonVisibilitySwitcherSlot}
  }
  Overlay_BusWarning {
   widgetClassPath=&Overlay
   bIsVariable=true
   parent=SizeBox_Default
   widget=Overlay_BusWarning
   properties {renderTransform {angle=0, scale=<1 1>, shear=<0 0>, translation=<0 -105>}}
   slot {class=SizeBoxSlot}
  }
  Image_SkewBG_Shadow {
   widgetClassPath=&Image
   parent=Overlay_BusWarning
   widget=Image_SkewBG_Shadow
   properties {
    visibility=SelfHitTestInvisible
    brush {
     bIsDynamicallyLoaded=false
     drawAs=Box
     imageType=NoImage
     mirroring=NoMirror
     resourceName=None
     resourceObject=&/M_UI_SkewBox_StormPhaseBG_Shadow
     tiling=NoTile
     imageSize=<64 64>
     margin=<ltrb 0.3 0 0.4 0>
     outlineSettings {
      bUseBrushTransparency=false
      roundingType=HalfHeightRadius
      width=0
      color {colorUseRule=UseColor_Specified, specifiedColor=<rgba 0 0 0 0>}
      cornerRadii=<0 0 0 0>
     }
     tintColor {colorUseRule=UseColor_Specified, specifiedColor=<rgba 0.0052 0.0052 0.0056 0.6>}
     uVRegion {bIsValid=false, max=<0 0>, min=<0 0>}
    }
    renderTransform {angle=0, scale=<1 1>, shear=<0 0>, translation=<8 8>}
   }
   slot {class=OverlaySlot, properties {horizontalAlignment=HAlign_Fill, verticalAlignment=VAlign_Fill, padding=<ltrb -25 -10 -5 0>}}
  }
  Image_SkewBG {
   widgetClassPath=&Image
   parent=Overlay_BusWarning
   widget=Image_SkewBG
   properties {
    visibility=SelfHitTestInvisible
    brush {
     bIsDynamicallyLoaded=false
     drawAs=Box
     imageType=NoImage
     mirroring=NoMirror
     resourceName=None
     resourceObject=&/M_UI_SkewBox_StormPhaseBG
     tiling=NoTile
     imageSize=<64 64>
     margin=<ltrb 0.3 0 0.4 0>
     outlineSettings {
      bUseBrushTransparency=false
      roundingType=HalfHeightRadius
      width=0
      color {colorUseRule=UseColor_Specified, specifiedColor=<rgba 0 0 0 0>}
      cornerRadii=<0 0 0 0>
     }
     tintColor {colorUseRule=UseColor_Specified, specifiedColor=<rgba 1 1 1 1>}
     uVRegion {bIsValid=false, max=<0 0>, min=<0 0>}
    }
   }
   slot {class=OverlaySlot, properties {horizontalAlignment=HAlign_Fill, verticalAlignment=VAlign_Fill, padding=<ltrb -25 -10 -5 0>}}
  }
  Switcher_BusIconState {
   widgetClassPath=&CommonVisibilitySwitcher
   bInherited=true
   bIsVariable=true
   parent=Overlay_BusWarning
   widget=Switcher_BusIconState
   slot {class=OverlaySlot, properties {horizontalAlignment=HAlign_Right, verticalAlignment=VAlign_Center, padding=<ltrb 0 -128 0 -128>}}
  }
  Image_Bus {
   widgetClassPath=&Image
   parent=Switcher_BusIconState
   widget=Image_Bus
   properties {
    visibility=SelfHitTestInvisible
    brush {
     bIsDynamicallyLoaded=false
     drawAs=Image
     imageType=NoImage
     mirroring=NoMirror
     resourceName=None
     resourceObject=/Game/Athena/HUD/Materials/M_UI_BusIcon
     tiling=NoTile
     imageSize=<167 148>
     margin=<ltrb 0 0 0 0>
     outlineSettings {
      bUseBrushTransparency=false
      roundingType=HalfHeightRadius
      width=0
      color {colorUseRule=UseColor_Specified, specifiedColor=<rgba 0 0 0 0>}
      cornerRadii=<0 0 0 0>
     }
     tintColor {colorUseRule=UseColor_Specified, specifiedColor=<rgba 1 1 1 1>}
     uVRegion {bIsValid=false, max=<0 0>, min=<0 0>}
    }
    renderTransform {angle=0, scale=<1 1>, shear=<0 0>, translation=<9 -18>}
   }
   slot {class=CommonVisibilitySwitcherSlot, properties {horizontalAlignment=HAlign_Center, verticalAlignment=VAlign_Center}}
  }
  Image_Skydive {
   widgetClassPath=&Image
   parent=Switcher_BusIconState
   widget=Image_Skydive
   properties {
    visibility=SelfHitTestInvisible
    brush {
     bIsDynamicallyLoaded=false
     drawAs=Image
     imageType=NoImage
     mirroring=NoMirror
     resourceName=None
     resourceObject=/Game/Athena/HUD/Materials/M_UI_SkydiveIcon
     tiling=NoTile
     imageSize=<162.7932 129.458>
     margin=<ltrb 0 0 0 0>
     outlineSettings {
      bUseBrushTransparency=false
      roundingType=HalfHeightRadius
      width=0
      color {colorUseRule=UseColor_Specified, specifiedColor=<rgba 0 0 0 0>}
      cornerRadii=<0 0 0 0>
     }
     tintColor {colorUseRule=UseColor_Specified, specifiedColor=<rgba 1 1 1 1>}
     uVRegion {bIsValid=false, max=<0 0>, min=<0 0>}
    }
    renderTransform {angle=0, scale=<1 1>, shear=<0 0>, translation=<19 -10>}
   }
   slot {class=CommonVisibilitySwitcherSlot, properties {horizontalAlignment=HAlign_Center, verticalAlignment=VAlign_Center}}
  }
  HorizontalBox_0 {
   widgetClassPath=&HorizontalBox
   parent=Overlay_BusWarning
   widget=HorizontalBox_0
   slot {class=OverlaySlot, properties {padding=<ltrb 0 5 170 5>}}
  }
  VerticalBox_0 {widgetClassPath=&VerticalBox, parent=HorizontalBox_0, widget=VerticalBox_0, slot {class=HorizontalBoxSlot}}
  Text_BusLine1 {
   widgetClassPath=&CommonTextBlock
   bInherited=true
   bIsVariable=true
   parent=VerticalBox_0
   widget=Text_BusLine1
   properties {
    text=Doors will
    textTransformPolicy=ToUpper
    font {
     bForceMonospaced=false
     bMaterialIsStencil=false
     fontMaterial=null
     fontObject=&BurbankSmall
     letterSpacing=0
     monospacedWidth=1
     size=18
     skewAmount=0
     typefaceFontName=Bold
     outlineSettings {
      bApplyOutlineToDropShadows=false
      bMiteredCorners=false
      bSeparateFillAlpha=false
      outlineMaterial=null
      outlineSize=0
      outlineColor=<rgba 0 0 0 1>
     }
    }
    renderTransform {angle=0, scale=<1 1>, shear=<-12 0>, translation=<0 0>}
    shadowOffset=<0 0>
   }
   slot {class=VerticalBoxSlot, properties {padding=<ltrb 0 0 0 -4>}}
  }
  Text_BusLine2 {
   widgetClassPath=&CommonTextBlock
   bInherited=true
   bIsVariable=true
   parent=VerticalBox_0
   widget=Text_BusLine2
   properties {
    justification=Right
    text=Open In
    textTransformPolicy=ToUpper
    font {
     bForceMonospaced=false
     bMaterialIsStencil=false
     fontMaterial=null
     fontObject=&BurbankSmall
     letterSpacing=0
     monospacedWidth=1
     size=24
     skewAmount=0
     typefaceFontName=Black
     outlineSettings {
      bApplyOutlineToDropShadows=false
      bMiteredCorners=false
      bSeparateFillAlpha=false
      outlineMaterial=null
      outlineSize=0
      outlineColor=<rgba 0 0 0 1>
     }
    }
    renderTransform {angle=0, scale=<1 1>, shear=<-12 0>, translation=<0 0>}
    shadowOffset=<0 0>
   }
   slot {class=VerticalBoxSlot, properties {padding=<ltrb 0 -4 0 0>}}
  }
  CommonVisualAttachment_0 {
   widgetClassPath=&CommonVisualAttachment
   parent=HorizontalBox_0
   widget=CommonVisualAttachment_0
   slot {class=HorizontalBoxSlot, properties {horizontalAlignment=HAlign_Left, verticalAlignment=VAlign_Center}}
  }
  Text_BusTimer {
   widgetClassPath=&CommonTextBlock
   bInherited=true
   bIsVariable=true
   parent=CommonVisualAttachment_0
   widget=Text_BusTimer
   properties {
    text=15s
    font {
     bForceMonospaced=false
     bMaterialIsStencil=false
     fontMaterial=null
     fontObject=&Burbank
     letterSpacing=0
     monospacedWidth=1
     size=60
     skewAmount=0
     typefaceFontName=Burbank-Black
     outlineSettings {
      bApplyOutlineToDropShadows=false
      bMiteredCorners=false
      bSeparateFillAlpha=false
      outlineMaterial=null
      outlineSize=3
      outlineColor=<rgba 0 0.1413 1 1>
     }
    }
    renderTransform {angle=0, scale=<1 1>, shear=<-12 0>, translation=<7 -22>}
    shadowOffset=<0 0>
   }
   slot {class=SizeBoxSlot}
  }
  WBP_RankedGamePhase {
   widgetClassPath=&/WBP_RankedGamePhase
   bIsVariable=true
   parent=BusWarning
   widget=WBP_RankedGamePhase
   properties {renderTransform {angle=0, scale=<1 1>, shear=<0 0>, translation=<0 -150>}}
   slot {class=CommonVisibilitySwitcherSlot}
  }
  ServerStatus {
   widgetClassPath=/Game/Athena/HUD/Papaya/Papaya_ServerWarning
   bIsVariable=true
   parent=Switcher_AlertType
   widget=ServerStatus
   properties {
    hideDiamondPulse=true
    borderColor=<rgba 0 0.8323 1 1>
    elementTag {tagName=HUD.Athena.PhaseChange}
    renderTransform {angle=0, scale=<1 1>, shear=<0 0>, translation=<0 -180>}
   }
   slot {class=CommonVisibilitySwitcherSlot}
  }
  TutorialHighlight {widgetClassPath=&TutorialHighlight}
 ]
}
