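# Translucent UI material for the storm-holding game phase in the Athena HUD, showing a clock over a tinted background.
# A Color vector parameter, cyan-blue by default, multiplies the phase background texture.
# A lerp blends that result toward the clock texture, using the clock's own RGB as both target and per-channel alpha, so its black areas leave the background visible.
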Material /Game/Athena/HUD/Phase/M_GamePhase_StormHolding
graph {
 properties {blendMode=BLEND_Translucent, materialDomain=MD_UI, pixelDepthOffsetMode=PDOM_Legacy, usedShadingModels=Default Lit}
 nodes [
  n1 = TextureSample() {texture=&/T_Clock}
  LinearInterpolate(
   A=Multiply(
    A=VectorParameter() {parameterName=Color, defaultValue=<rgba 0 0.655 0.9216 1>}.RGB
    B=TextureSample() {texture=&/T_GamePhaseBackground}.RGB
   )
   B=@n1.RGB
   Alpha=@n1.RGB
  )
 ]
}
