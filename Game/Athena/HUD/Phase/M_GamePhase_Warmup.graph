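# UI material for the Warmup game phase icon in the Athena HUD, showing a clock over a tinted phase background.
# It multiplies the game phase background texture by a constant blue color.
# A lerp blends that tinted background toward the clock texture, using the clock's own RGB as both the target color and the per-channel alpha.
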Material /Game/Athena/HUD/Phase/M_GamePhase_Warmup
graph {
 properties {blendMode=BLEND_Translucent, materialDomain=MD_UI, pixelDepthOffsetMode=PDOM_Legacy, usedShadingModels=Default Lit}
 nodes [
  n1 = TextureSample() {texture=&/T_Clock}
  LinearInterpolate(
   A=Multiply(A=Constant3Vector() {constant=<rgba 0.0561 0.0908 0.5457 1>}, B=TextureSample() {texture=&/T_GamePhaseBackground}.RGB)
   B=@n1.RGB
   Alpha=@n1.RGB
  )
 ]
}
