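# HUD widget for the Athena game phase, showing a phase icon beside a remaining-time text.
# On each phase icon update it resets to a default timer material, then selects a material by phase step, including bus and storm stages.
# That selection only runs when the owning player's minigame is absent or in progress and the step is not None.
# A second sequence branch updates the Image brush only when the material differs.
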WidgetBlueprint /Game/Athena/HUD/Phase/AthenaGamePhaseWidget
graphs [EventGraph {nodes [
   n1 = Event OnUpdateGamePhaseIcon!() -> (OutputDelegate, then, GamePhaseStep:EAthenaGamePhaseStep Enum=None) {is=title K2Node_Event}
   > Set GamePhaseIcon(GamePhaseIcon:Material Interface Object Reference=Get DefaultGamePhaseIcon() -> (DefaultGamePhaseIcon:Material Interface Object Reference)) -> (Output_Get:Material Interface Object Reference)
   > n2 = Sequence()
   n3 = GetOwningPlayer()
   > Is Valid(InputObject=@n3)
   n4 = GetMinigame!(ContextActor:Actor Object Reference=@n3) -> (ReturnValue:Fort Minigame Object Reference)
   > n5 = Is Valid(InputObject=@n4)
   > n6 = Branch(Condition="Equal (Enum)"(
     A:EFortMinigameState Enum=Get CurrentState!(self:Fort Minigame Object Reference=@n4) -> (CurrentState:EFortMinigameState Enum=PreGame)
     B:EFortMinigameState Enum=InProgress
    ))
   Branch(
    execute=[@n6.then, @n5.Is Not Valid]
    Condition="Not Equal (Enum)"!(A:EAthenaGamePhaseStep Enum=@n1.GamePhaseStep, B:EAthenaGamePhaseStep Enum=None) -> (ReturnValue:bool)
   )
   > Set GamePhaseIcon(GamePhaseIcon=Select!(
     None:Material Interface Object Reference
     Setup:Material Interface Object Reference=&/M_GamePhase_Setup
     Warmup:Material Interface Object Reference=&/M_GamePhase_Warmup
     GetReady:Material Interface Object Reference=&/M_GamePhase_Bus
     BusLocked:Material Interface Object Reference=&M_GamePhase_Bus
     BusFlying:Material Interface Object Reference=&/M_GamePhase_Jump
     StormForming:Material Interface Object Reference=/Game/Athena/HUD/Materials/M_UI_Timer_S
     StormHolding:Material Interface Object Reference=&M_UI_Timer_S
     StormShrinking:Material Interface Object Reference=&/M_GamePhase_StormMoving
     Countdown:Material Interface Object Reference=&/M_GamePhase_StormHolding
     FinalCountdown:Material Interface Object Reference=&M_GamePhase_StormHolding
     EndGame:Material Interface Object Reference=&/M_GamePhase_End
     Count:Material Interface Object Reference
     Index:EAthenaGamePhaseStep Enum=@n1.GamePhaseStep
    ) -> (ReturnValue:Material Interface Object Reference))
   n7 = Get Brush!(self:Image Object Reference=Get Icon() -> (Icon:Image Object Reference)) -> (Brush:Slate Brush Structure)
   Branch(
    execute=@n2.then_1
    Condition="Not Equal (Object)"(
     A:Object Reference=GetBrushResourceAsMaterial!(Brush:"Slate Brush Structure (by ref)"=@n7) -> (ReturnValue:Material Interface Object Reference)
     B:Object Reference=Get GamePhaseIcon() -> (GamePhaseIcon:Material Interface Object Reference)
    ) -> (ReturnValue:bool=~) {is=title K2Node_PromotableOperator}
   )
   > SetBrushFromMaterial(self=Get Icon(), Material=Get GamePhaseIcon())
  ]}]
variables [
 DefaultGamePhaseIcon {type=MaterialInterface, default=&M_UI_Timer_S}
 GamePhaseIcon {type=MaterialInterface, default=null}
]
widgets {
 info {inheritedWidgetCount=2, parentClass=/Script/FortniteUI.AthenaGamePhaseWidgetBase, rootWidgetClass=&HorizontalBox, widgetCount=4}
 widgets [
  HorizontalBox_0 {widgetClassPath=&HorizontalBox, widget=HorizontalBox_0}
  Icon {
   widgetClassPath=&Image
   bIsVariable=true
   parent=HorizontalBox_0
   widget=Icon
   properties {
    visibility=SelfHitTestInvisible
    brush {
     bIsDynamicallyLoaded=false
     drawAs=Image
     imageType=NoImage
     mirroring=NoMirror
     resourceName=None
     resourceObject=&M_UI_Timer_S
     tiling=NoTile
     imageSize=<36 36>
     margin=<ltrb 0 0 0 0>
     outlineSettings {
      bUseBrushTransparency=false
      roundingType=HalfHeightRadius
      width=0
      color {colorUseRule=UseColor_Specified, specifiedColor=<rgba 0 0 0 0>}
      cornerRadii=<0 0 0 0>
     }
     tintColor {colorUseRule=UseColor_Specified, specifiedColor=<rgba 1 1 1 1>}
     uVRegion {bIsValid=false, max=<0 0>, min=<0 0>}
    }
   }
   slot {class=HorizontalBoxSlot, properties {horizontalAlignment=HAlign_Center, verticalAlignment=VAlign_Center}}
  }
  Text_RemainingTime {
   widgetClassPath=&TextBlock
   bInherited=true
   parent=HorizontalBox_0
   widget=Text_RemainingTime
   properties {
    bIsVolatile=true
    justification=Center
    minDesiredWidth=50
    text=0:00
    visibility=SelfHitTestInvisible
    font {
     bForceMonospaced=false
     bMaterialIsStencil=false
     fontMaterial=null
     fontObject=&BurbankSmall
     letterSpacing=0
     monospacedWidth=1
     size=20
     skewAmount=0
     typefaceFontName=Bold
     outlineSettings {
      bApplyOutlineToDropShadows=false
      bMiteredCorners=false
      bSeparateFillAlpha=true
      outlineMaterial=&M-FontShadow
      outlineSize=3
      outlineColor=<rgba 0.0052 0.0052 0.0056 1>
     }
    }
    shadowOffset=<0 0>
   }
   slot {class=HorizontalBoxSlot, properties {verticalAlignment=VAlign_Center, padding=<ltrb 4 0 0 0>}}
  }
  TutorialHighlight {widgetClassPath=&TutorialHighlight}
 ]
}
